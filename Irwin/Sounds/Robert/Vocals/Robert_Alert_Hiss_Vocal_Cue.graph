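# Sound cue for Robert's alert hiss creature vocal, playing random Idle_Aggressive vocal waves with medium predator vocalization attenuation and a capped concurrency override.
# IsLocal and IsRiddenByEnemy branches select local, enemy-ridden or other paths, which currently draw on the same waves at equal mixer volume.
# Within each path, a quality-level node picks a 13-wave random pool for the first two levels and a four-wave subset for the last two.
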SoundCue /Irwin/Sounds/Robert/Vocals/Robert_Alert_Hiss_Vocal_Cue
root=Branch_1
nodes [
 {id=Branch_0, children [Mixer_1,Mixer_0,Mixer_1], properties {boolParameterName=IsRiddenByEnemy}}
 {id=Branch_1, children [Mixer_2,Branch_0,Branch_0], properties {boolParameterName=IsLocal}}
 {id=Mixer_0, children [QualityLevel_2], properties {inputVolume [1]}}
 {id=Mixer_1, children [QualityLevel_1], properties {inputVolume [1]}}
 {id=Mixer_2, children [QualityLevel_0], properties {inputVolume [1]}}
 {id=QualityLevel_0, children [Random_4,Random_4,Random_5,Random_5]}
 {id=QualityLevel_1, children [Random_0,Random_0,Random_1,Random_1]}
 {id=QualityLevel_2, children [Random_2,Random_2,Random_3,Random_3]}
 {
  id=Random_0
  children [
   WavePlayer_0
   WavePlayer_1
   WavePlayer_2
   WavePlayer_8
   WavePlayer_9
   WavePlayer_10
   WavePlayer_11
   WavePlayer_13
   WavePlayer_14
   WavePlayer_15
   WavePlayer_16
   WavePlayer_28
   WavePlayer_29
  ]
  properties {bShouldExcludeFromBranchCulling=true, weights [1,1,1,1,1,1,1,1,1,1,1,1,1]}
 }
 {
  id=Random_1
  children [WavePlayer_31, WavePlayer_32, WavePlayer_33, WavePlayer_30]
  properties {bShouldExcludeFromBranchCulling=true, weights [1,1,1,1]}
 }
 {
  id=Random_2
  children [
   WavePlayer_34
   WavePlayer_35
   WavePlayer_36
   WavePlayer_37
   WavePlayer_38
   WavePlayer_39
   WavePlayer_40
   WavePlayer_41
   WavePlayer_42
   WavePlayer_43
   WavePlayer_44
   WavePlayer_45
   WavePlayer_46
  ]
  properties {bShouldExcludeFromBranchCulling=true, weights [1,1,1,1,1,1,1,1,1,1,1,1,1]}
 }
 {
  id=Random_3
  children [WavePlayer_48, WavePlayer_49, WavePlayer_50, WavePlayer_47]
  properties {bShouldExcludeFromBranchCulling=true, weights [1,1,1,1]}
 }
 {
  id=Random_4
  children [
   WavePlayer_3
   WavePlayer_4
   WavePlayer_5
   WavePlayer_6
   WavePlayer_7
   WavePlayer_12
   WavePlayer_17
   WavePlayer_18
   WavePlayer_19
   WavePlayer_20
   WavePlayer_21
   WavePlayer_22
   WavePlayer_23
  ]
  properties {bShouldExcludeFromBranchCulling=true, weights [1,1,1,1,1,1,1,1,1,1,1,1,1]}
 }
 {
  id=Random_5
  children [WavePlayer_25, WavePlayer_26, WavePlayer_27, WavePlayer_24]
  properties {bShouldExcludeFromBranchCulling=true, weights [1,1,1,1]}
 }
 {id=WavePlayer_0, properties {soundWaveAssetPtr=&/Robert_Idle_Aggressive_Vocal_10}}
 {id=WavePlayer_1, properties {soundWaveAssetPtr=&/Robert_Idle_Aggressive_Vocal_11}}
 {id=WavePlayer_10, properties {soundWaveAssetPtr=&/Robert_Idle_Aggressive_Vocal_02}}
 {id=WavePlayer_11, properties {soundWaveAssetPtr=&/Robert_Idle_Aggressive_Vocal_03}}
 {id=WavePlayer_12, properties {soundWaveAssetPtr=&Robert_Idle_Aggressive_Vocal_02}}
 {id=WavePlayer_13, properties {soundWaveAssetPtr=&/Robert_Idle_Aggressive_Vocal_04}}
 {id=WavePlayer_14, properties {soundWaveAssetPtr=&/Robert_Idle_Aggressive_Vocal_05}}
 {id=WavePlayer_15, properties {soundWaveAssetPtr=&/Robert_Idle_Aggressive_Vocal_06}}
 {id=WavePlayer_16, properties {soundWaveAssetPtr=&/Robert_Idle_Aggressive_Vocal_07}}
 {id=WavePlayer_17, properties {soundWaveAssetPtr=&Robert_Idle_Aggressive_Vocal_03}}
 {id=WavePlayer_18, properties {soundWaveAssetPtr=&Robert_Idle_Aggressive_Vocal_04}}
 {id=WavePlayer_19, properties {soundWaveAssetPtr=&Robert_Idle_Aggressive_Vocal_05}}
 {id=WavePlayer_2, properties {soundWaveAssetPtr=&/Robert_Idle_Aggressive_Vocal_12}}
 {id=WavePlayer_20, properties {soundWaveAssetPtr=&Robert_Idle_Aggressive_Vocal_06}}
 {id=WavePlayer_21, properties {soundWaveAssetPtr=&Robert_Idle_Aggressive_Vocal_07}}
 {id=WavePlayer_22, properties {soundWaveAssetPtr=&/Robert_Idle_Aggressive_Vocal_08}}
 {id=WavePlayer_23, properties {soundWaveAssetPtr=&/Robert_Idle_Aggressive_Vocal_09}}
 {id=WavePlayer_24, properties {soundWaveAssetPtr=&Robert_Idle_Aggressive_Vocal_05}}
 {id=WavePlayer_25, properties {soundWaveAssetPtr=&Robert_Idle_Aggressive_Vocal_10}}
 {id=WavePlayer_26, properties {soundWaveAssetPtr=&Robert_Idle_Aggressive_Vocal_11}}
 {id=WavePlayer_27, properties {soundWaveAssetPtr=&Robert_Idle_Aggressive_Vocal_12}}
 {id=WavePlayer_28, properties {soundWaveAssetPtr=&Robert_Idle_Aggressive_Vocal_08}}
 {id=WavePlayer_29, properties {soundWaveAssetPtr=&Robert_Idle_Aggressive_Vocal_09}}
 {id=WavePlayer_3, properties {soundWaveAssetPtr=&Robert_Idle_Aggressive_Vocal_10}}
 {id=WavePlayer_30, properties {soundWaveAssetPtr=&Robert_Idle_Aggressive_Vocal_05}}
 {id=WavePlayer_31, properties {soundWaveAssetPtr=&Robert_Idle_Aggressive_Vocal_10}}
 {id=WavePlayer_32, properties {soundWaveAssetPtr=&Robert_Idle_Aggressive_Vocal_11}}
 {id=WavePlayer_33, properties {soundWaveAssetPtr=&Robert_Idle_Aggressive_Vocal_12}}
 {id=WavePlayer_34, properties {soundWaveAssetPtr=&Robert_Idle_Aggressive_Vocal_10}}
 {id=WavePlayer_35, properties {soundWaveAssetPtr=&Robert_Idle_Aggressive_Vocal_11}}
 {id=WavePlayer_36, properties {soundWaveAssetPtr=&Robert_Idle_Aggressive_Vocal_12}}
 {id=WavePlayer_37, properties {soundWaveAssetPtr=&/Robert_Idle_Aggressive_Vocal_13}}
 {id=WavePlayer_38, properties {soundWaveAssetPtr=&/Robert_Idle_Aggressive_Vocal_01}}
 {id=WavePlayer_39, properties {soundWaveAssetPtr=&Robert_Idle_Aggressive_Vocal_02}}
 {id=WavePlayer_4, properties {soundWaveAssetPtr=&Robert_Idle_Aggressive_Vocal_11}}
 {id=WavePlayer_40, properties {soundWaveAssetPtr=&Robert_Idle_Aggressive_Vocal_03}}
 {id=WavePlayer_41, properties {soundWaveAssetPtr=&Robert_Idle_Aggressive_Vocal_04}}
 {id=WavePlayer_42, properties {soundWaveAssetPtr=&Robert_Idle_Aggressive_Vocal_05}}
 {id=WavePlayer_43, properties {soundWaveAssetPtr=&Robert_Idle_Aggressive_Vocal_06}}
 {id=WavePlayer_44, properties {soundWaveAssetPtr=&Robert_Idle_Aggressive_Vocal_07}}
 {id=WavePlayer_45, properties {soundWaveAssetPtr=&Robert_Idle_Aggressive_Vocal_08}}
 {id=WavePlayer_46, properties {soundWaveAssetPtr=&Robert_Idle_Aggressive_Vocal_09}}
 {id=WavePlayer_47, properties {soundWaveAssetPtr=&Robert_Idle_Aggressive_Vocal_05}}
 {id=WavePlayer_48, properties {soundWaveAssetPtr=&Robert_Idle_Aggressive_Vocal_10}}
 {id=WavePlayer_49, properties {soundWaveAssetPtr=&Robert_Idle_Aggressive_Vocal_11}}
 {id=WavePlayer_5, properties {soundWaveAssetPtr=&Robert_Idle_Aggressive_Vocal_12}}
 {id=WavePlayer_50, properties {soundWaveAssetPtr=&Robert_Idle_Aggressive_Vocal_12}}
 {id=WavePlayer_6, properties {soundWaveAssetPtr=&Robert_Idle_Aggressive_Vocal_13}}
 {id=WavePlayer_7, properties {soundWaveAssetPtr=&Robert_Idle_Aggressive_Vocal_01}}
 {id=WavePlayer_8, properties {soundWaveAssetPtr=&Robert_Idle_Aggressive_Vocal_13}}
 {id=WavePlayer_9, properties {soundWaveAssetPtr=&Robert_Idle_Aggressive_Vocal_01}}
]
properties {
 attenuationSettings=/Irwin/Sounds/Settings/Predator_Vocalization_Medium_Att
 bOverrideConcurrency=true
 maxDistance=12000
 soundClassObject=&Fort_Creatures_Vocalization_Tactical
 volumeMultiplier=2
 concurrencyOverrides {maxCount=3}
}
