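# Sound cue for the creature Robert's alert vocalization, playing a random pre-bark wave with medium predator attenuation and tactical creature vocalization sound class.
# It branches on IsLocal, then on IsRiddenByEnemy for non-local or unset cases, with an unset rider flag following the enemy path.
# Each of the three paths runs through a mixer, delay and quality-level switch, choosing from six variants or a reduced pair.
# All paths share the same waves.
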SoundCue /Irwin/Sounds/Robert/Vocals/Robert_Alert_Vocal_Alert_Cue
root=Branch_1
nodes [
 {id=Branch_0, children [Mixer_1,Mixer_0,Mixer_1], properties {boolParameterName=IsRiddenByEnemy}}
 {id=Branch_1, children [Mixer_2,Branch_0,Branch_0], properties {boolParameterName=IsLocal}}
 {id=Delay_0, children [QualityLevel_0]}
 {id=Delay_1, children [QualityLevel_1]}
 {id=Delay_2, children [QualityLevel_2]}
 {id=Mixer_0, children [Delay_2], properties {inputVolume [1]}}
 {id=Mixer_1, children [Delay_1], properties {inputVolume [1]}}
 {id=Mixer_2, children [Delay_0], properties {inputVolume [1]}}
 {id=QualityLevel_0, children [Random_1,Random_1,Random_0,Random_0]}
 {id=QualityLevel_1, children [Random_2,Random_2,Random_3,Random_3]}
 {id=QualityLevel_2, children [Random_4,Random_4,Random_5,Random_5]}
 {id=Random_0, children [WavePlayer_0,WavePlayer_1], properties {bShouldExcludeFromBranchCulling=true, weights [1,1]}}
 {
  id=Random_1
  children [WavePlayer_6, WavePlayer_7, WavePlayer_8, WavePlayer_9, WavePlayer_10, WavePlayer_11]
  properties {bShouldExcludeFromBranchCulling=true, weights [1,1,1,1,1,1]}
 }
 {
  id=Random_2
  children [WavePlayer_2, WavePlayer_3, WavePlayer_4, WavePlayer_5, WavePlayer_12, WavePlayer_13]
  properties {bShouldExcludeFromBranchCulling=true, weights [1,1,1,1,1,1]}
 }
 {id=Random_3, children [WavePlayer_14, WavePlayer_15], properties {bShouldExcludeFromBranchCulling=true, weights [1,1]}}
 {
  id=Random_4
  children [WavePlayer_16, WavePlayer_17, WavePlayer_18, WavePlayer_19, WavePlayer_20, WavePlayer_21]
  properties {bShouldExcludeFromBranchCulling=true, weights [1,1,1,1,1,1]}
 }
 {id=Random_5, children [WavePlayer_22, WavePlayer_23], properties {bShouldExcludeFromBranchCulling=true, weights [1,1]}}
 {id=WavePlayer_0, properties {soundWaveAssetPtr=&/Robert_Alert_Vocal_PreBark_01}}
 {id=WavePlayer_1, properties {soundWaveAssetPtr=&/Robert_Alert_Vocal_PreBark_02}}
 {id=WavePlayer_10, properties {soundWaveAssetPtr=&/Robert_Alert_Vocal_PreBark_05}}
 {id=WavePlayer_11, properties {soundWaveAssetPtr=&/Robert_Alert_Vocal_PreBark_06}}
 {id=WavePlayer_12, properties {soundWaveAssetPtr=&Robert_Alert_Vocal_PreBark_05}}
 {id=WavePlayer_13, properties {soundWaveAssetPtr=&Robert_Alert_Vocal_PreBark_06}}
 {id=WavePlayer_14, properties {soundWaveAssetPtr=&Robert_Alert_Vocal_PreBark_01}}
 {id=WavePlayer_15, properties {soundWaveAssetPtr=&Robert_Alert_Vocal_PreBark_02}}
 {id=WavePlayer_16, properties {soundWaveAssetPtr=&Robert_Alert_Vocal_PreBark_01}}
 {id=WavePlayer_17, properties {soundWaveAssetPtr=&Robert_Alert_Vocal_PreBark_02}}
 {id=WavePlayer_18, properties {soundWaveAssetPtr=&/Robert_Alert_Vocal_PreBark_03}}
 {id=WavePlayer_19, properties {soundWaveAssetPtr=&/Robert_Alert_Vocal_PreBark_04}}
 {id=WavePlayer_2, properties {soundWaveAssetPtr=&Robert_Alert_Vocal_PreBark_01}}
 {id=WavePlayer_20, properties {soundWaveAssetPtr=&Robert_Alert_Vocal_PreBark_05}}
 {id=WavePlayer_21, properties {soundWaveAssetPtr=&Robert_Alert_Vocal_PreBark_06}}
 {id=WavePlayer_22, properties {soundWaveAssetPtr=&Robert_Alert_Vocal_PreBark_01}}
 {id=WavePlayer_23, properties {soundWaveAssetPtr=&Robert_Alert_Vocal_PreBark_02}}
 {id=WavePlayer_3, properties {soundWaveAssetPtr=&Robert_Alert_Vocal_PreBark_02}}
 {id=WavePlayer_4, properties {soundWaveAssetPtr=&Robert_Alert_Vocal_PreBark_03}}
 {id=WavePlayer_5, properties {soundWaveAssetPtr=&Robert_Alert_Vocal_PreBark_04}}
 {id=WavePlayer_6, properties {soundWaveAssetPtr=&Robert_Alert_Vocal_PreBark_01}}
 {id=WavePlayer_7, properties {soundWaveAssetPtr=&Robert_Alert_Vocal_PreBark_02}}
 {id=WavePlayer_8, properties {soundWaveAssetPtr=&Robert_Alert_Vocal_PreBark_03}}
 {id=WavePlayer_9, properties {soundWaveAssetPtr=&Robert_Alert_Vocal_PreBark_04}}
]
properties {
 attenuationSettings=/Irwin/Sounds/Settings/Predator_Vocalization_Medium_Att
 bOverrideConcurrency=true
 maxDistance=12000
 soundClassObject=&Fort_Creatures_Vocalization_Tactical
 volumeMultiplier=1.25
 concurrencyOverrides {maxCount=3}
}
