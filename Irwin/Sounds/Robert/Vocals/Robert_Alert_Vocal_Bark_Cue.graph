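# Sound cue for Robert's alert bark vocalization, a creature sound with predator medium-range attenuation and a concurrency cap.
# An IsLocal branch plays the local pool or, otherwise, defers to an IsRiddenByEnemy branch, where an unset value shares the ridden-by-enemy pool.
# Each of the three pools randomly picks among the same six bark variations at equal weight and mixer volume, with the random nodes excluded from branch culling.
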SoundCue /Irwin/Sounds/Robert/Vocals/Robert_Alert_Vocal_Bark_Cue
root=Branch_1
nodes [
 {id=Branch_0, children [Mixer_1,Mixer_0,Mixer_1], properties {boolParameterName=IsRiddenByEnemy}}
 {id=Branch_1, children [Mixer_2,Branch_0,Branch_0], properties {boolParameterName=IsLocal}}
 {id=Mixer_0, children [Random_3], properties {inputVolume [1]}}
 {id=Mixer_1, children [Random_2], properties {inputVolume [1]}}
 {id=Mixer_2, children [Random_0], properties {inputVolume [1]}}
 {
  id=Random_0
  children [WavePlayer_0,WavePlayer_1,WavePlayer_2,WavePlayer_3,WavePlayer_4,WavePlayer_5]
  properties {bShouldExcludeFromBranchCulling=true, weights [1,1,1,1,1,1]}
 }
 {
  id=Random_2
  children [WavePlayer_6, WavePlayer_8, WavePlayer_9, WavePlayer_10, WavePlayer_12, WavePlayer_13]
  properties {bShouldExcludeFromBranchCulling=true, weights [1,1,1,1,1,1]}
 }
 {
  id=Random_3
  children [WavePlayer_14, WavePlayer_15, WavePlayer_16, WavePlayer_17, WavePlayer_18, WavePlayer_19]
  properties {bShouldExcludeFromBranchCulling=true, weights [1,1,1,1,1,1]}
 }
 {id=WavePlayer_0, properties {soundWaveAssetPtr=&/Robert_Alert_Vocal_Bark_06}}
 {id=WavePlayer_1, properties {soundWaveAssetPtr=&/Robert_Alert_Vocal_Bark_01}}
 {id=WavePlayer_10, properties {soundWaveAssetPtr=&/Robert_Alert_Vocal_Bark_03}}
 {id=WavePlayer_12, properties {soundWaveAssetPtr=&/Robert_Alert_Vocal_Bark_04}}
 {id=WavePlayer_13, properties {soundWaveAssetPtr=&/Robert_Alert_Vocal_Bark_05}}
 {id=WavePlayer_14, properties {soundWaveAssetPtr=&Robert_Alert_Vocal_Bark_06}}
 {id=WavePlayer_15, properties {soundWaveAssetPtr=&Robert_Alert_Vocal_Bark_01}}
 {id=WavePlayer_16, properties {soundWaveAssetPtr=&/Robert_Alert_Vocal_Bark_02}}
 {id=WavePlayer_17, properties {soundWaveAssetPtr=&Robert_Alert_Vocal_Bark_03}}
 {id=WavePlayer_18, properties {soundWaveAssetPtr=&Robert_Alert_Vocal_Bark_04}}
 {id=WavePlayer_19, properties {soundWaveAssetPtr=&Robert_Alert_Vocal_Bark_05}}
 {id=WavePlayer_2, properties {soundWaveAssetPtr=&Robert_Alert_Vocal_Bark_02}}
 {id=WavePlayer_3, properties {soundWaveAssetPtr=&Robert_Alert_Vocal_Bark_03}}
 {id=WavePlayer_4, properties {soundWaveAssetPtr=&Robert_Alert_Vocal_Bark_04}}
 {id=WavePlayer_5, properties {soundWaveAssetPtr=&Robert_Alert_Vocal_Bark_05}}
 {id=WavePlayer_6, properties {soundWaveAssetPtr=&Robert_Alert_Vocal_Bark_06}}
 {id=WavePlayer_8, properties {soundWaveAssetPtr=&Robert_Alert_Vocal_Bark_01}}
 {id=WavePlayer_9, properties {soundWaveAssetPtr=&Robert_Alert_Vocal_Bark_02}}
]
properties {
 attenuationSettings=/Irwin/Sounds/Settings/Predator_Vocalization_Medium_Att
 bOverrideConcurrency=true
 maxDistance=12000
 soundClassObject=&Fort_Creatures_Vocalization_Tactical
 volumeMultiplier=1.8
 concurrencyOverrides {maxCount=3}
}
