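# Sound cue for the exhale breathing vocal of Robert, a rideable creature.
# IsLocal plays at medium volume; otherwise IsRiddenByEnemy selects full volume when true or unset, quiet when false.
# Each path randomly picks mostly exhales, occasionally an idle chitter or silence via empty slots, using repeated inputs as weighting.
# Modulators vary pitch and volume, and per-owner and global concurrency sets govern overlapping breaths.
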SoundCue /Irwin/Sounds/Robert/Vocals/Robert_Breath_Exhale_Vocal_Cue
root=Branch_1
nodes [
 {id=Branch_0, children [Mixer_1,Mixer_0,Mixer_1], properties {boolParameterName=IsRiddenByEnemy}}
 {id=Branch_1, children [Mixer_2,Branch_0,Branch_0], properties {boolParameterName=IsLocal}}
 {id=Mixer_0, children [Random_7], properties {inputVolume [0.3]}}
 {id=Mixer_1, children [Random_2], properties {inputVolume [1]}}
 {id=Mixer_2, children [Random_3], properties {inputVolume [0.7]}}
 {id=Modulator_0, children [Random_1], properties {pitchMax=1.2, pitchMin=1, volumeMin=0.1}}
 {id=Modulator_1, children [Random_6], properties {pitchMax=1.2, pitchMin=1, volumeMax=1, volumeMin=0.75}}
 {id=Modulator_2, children [Random_0], properties {pitchMax=1.2, pitchMin=1, volumeMin=0.1}}
 {id=Modulator_3, children [Random_4], properties {pitchMax=1.2, pitchMin=1, volumeMax=1, volumeMin=0.75}}
 {id=Modulator_4, children [Random_5], properties {pitchMax=1.2, pitchMin=1, volumeMin=0.1}}
 {id=Modulator_5, children [Random_8], properties {pitchMax=1.2, pitchMin=1, volumeMax=1, volumeMin=0.75}}
 {
  id=Random_0
  children [
   WavePlayer_3
   WavePlayer_4
   WavePlayer_5
   WavePlayer_6
   WavePlayer_7
   WavePlayer_8
   WavePlayer_9
   WavePlayer_10
   WavePlayer_11
   WavePlayer_12
   WavePlayer_13
   WavePlayer_14
   WavePlayer_16
   WavePlayer_17
   WavePlayer_35
  ]
  properties {bShouldExcludeFromBranchCulling=true, weights [1,1,1,1,1,1,1,1,1,1,1,1,1,1,1]}
 }
 {
  id=Random_1
  children [
   WavePlayer_22
   WavePlayer_23
   WavePlayer_24
   WavePlayer_25
   WavePlayer_26
   WavePlayer_27
   WavePlayer_28
   WavePlayer_29
   WavePlayer_30
   WavePlayer_31
   WavePlayer_32
   WavePlayer_33
   WavePlayer_34
   WavePlayer_0
   WavePlayer_36
  ]
  properties {bShouldExcludeFromBranchCulling=true, weights [1,1,1,1,1,1,1,1,1,1,1,1,1,1,1]}
 }
 {
  id=Random_2
  children [Modulator_2,Modulator_2,Modulator_2,Modulator_2,Modulator_2,Modulator_2,Modulator_2,Modulator_3,null,null,null]
  properties {bShouldExcludeFromBranchCulling=true, weights [1,1,1,1,1,1,1,1,1,1,1]}
 }
 {
  id=Random_3
  children [Modulator_0,Modulator_0,Modulator_0,Modulator_0,Modulator_0,Modulator_0,Modulator_0,Modulator_1,null,null,null]
  properties {bShouldExcludeFromBranchCulling=true, weights [1,1,1,1,1,1,1,1,1,1,1]}
 }
 {
  id=Random_4
  children [
   WavePlayer_37
   WavePlayer_38
   WavePlayer_39
   WavePlayer_40
   WavePlayer_41
   WavePlayer_42
   WavePlayer_43
   WavePlayer_44
   WavePlayer_45
   WavePlayer_46
   WavePlayer_47
   WavePlayer_48
   WavePlayer_49
   WavePlayer_50
  ]
  properties {bShouldExcludeFromBranchCulling=true, weights [1,1,1,1,1,1,1,1,1,1,1,1,1,1]}
 }
 {
  id=Random_5
  children [
   WavePlayer_51
   WavePlayer_52
   WavePlayer_53
   WavePlayer_54
   WavePlayer_55
   WavePlayer_56
   WavePlayer_57
   WavePlayer_58
   WavePlayer_66
   WavePlayer_67
   WavePlayer_68
   WavePlayer_69
   WavePlayer_70
   WavePlayer_71
   WavePlayer_72
  ]
  properties {bShouldExcludeFromBranchCulling=true, weights [1,1,1,1,1,1,1,1,1,1,1,1,1,1,1]}
 }
 {
  id=Random_6
  children [
   WavePlayer_15
   WavePlayer_1
   WavePlayer_2
   WavePlayer_18
   WavePlayer_19
   WavePlayer_20
   WavePlayer_21
   WavePlayer_59
   WavePlayer_60
   WavePlayer_61
   WavePlayer_62
   WavePlayer_63
   WavePlayer_64
   WavePlayer_65
  ]
  properties {bShouldExcludeFromBranchCulling=true, weights [1,1,1,1,1,1,1,1,1,1,1,1,1,1]}
 }
 {
  id=Random_7
  children [Modulator_4,Modulator_4,Modulator_4,Modulator_4,Modulator_4,Modulator_4,Modulator_4,Modulator_5,null,null,null]
  properties {bShouldExcludeFromBranchCulling=true, weights [1,1,1,1,1,1,1,1,1,1,1]}
 }
 {
  id=Random_8
  children [
   WavePlayer_73
   WavePlayer_74
   WavePlayer_75
   WavePlayer_76
   WavePlayer_77
   WavePlayer_78
   WavePlayer_79
   WavePlayer_80
   WavePlayer_81
   WavePlayer_82
   WavePlayer_83
   WavePlayer_84
   WavePlayer_85
   WavePlayer_86
  ]
  properties {bShouldExcludeFromBranchCulling=true, weights [1,1,1,1,1,1,1,1,1,1,1,1,1,1]}
 }
 {id=WavePlayer_0, properties {soundWaveAssetPtr=&/Robert_Walk_Vocal_Exhale_08}}
 {id=WavePlayer_1, properties {soundWaveAssetPtr=&/Robert_Idle_Vocal_Chitter_11}}
 {id=WavePlayer_10, properties {soundWaveAssetPtr=&/Robert_Walk_Vocal_Exhale_02}}
 {id=WavePlayer_11, properties {soundWaveAssetPtr=&/Robert_Walk_Vocal_Exhale_03}}
 {id=WavePlayer_12, properties {soundWaveAssetPtr=&/Robert_Walk_Vocal_Exhale_04}}
 {id=WavePlayer_13, properties {soundWaveAssetPtr=&/Robert_Walk_Vocal_Exhale_05}}
 {id=WavePlayer_14, properties {soundWaveAssetPtr=&/Robert_Walk_Vocal_Exhale_06}}
 {id=WavePlayer_15, properties {soundWaveAssetPtr=&/Robert_Idle_Vocal_Chitter_10}}
 {id=WavePlayer_16, properties {soundWaveAssetPtr=&/Robert_Walk_Vocal_Exhale_07}}
 {id=WavePlayer_17, properties {soundWaveAssetPtr=&Robert_Walk_Vocal_Exhale_08}}
 {id=WavePlayer_18, properties {soundWaveAssetPtr=&/Robert_Idle_Vocal_Chitter_16}}
 {id=WavePlayer_19, properties {soundWaveAssetPtr=&/Robert_Idle_Vocal_Chitter_17}}
 {id=WavePlayer_2, properties {soundWaveAssetPtr=&/Robert_Idle_Vocal_Chitter_15}}
 {id=WavePlayer_20, properties {soundWaveAssetPtr=&/Robert_Idle_Vocal_Chitter_21}}
 {id=WavePlayer_21, properties {soundWaveAssetPtr=&/Robert_Idle_Vocal_Chitter_22}}
 {id=WavePlayer_22, properties {soundWaveAssetPtr=&/Robert_Walk_Vocal_Exhale_10}}
 {id=WavePlayer_23, properties {soundWaveAssetPtr=&/Robert_Walk_Vocal_Exhale_11}}
 {id=WavePlayer_24, properties {soundWaveAssetPtr=&/Robert_Walk_Vocal_Exhale_12}}
 {id=WavePlayer_25, properties {soundWaveAssetPtr=&/Robert_Walk_Vocal_Exhale_13}}
 {id=WavePlayer_26, properties {soundWaveAssetPtr=&/Robert_Walk_Vocal_Exhale_14}}
 {id=WavePlayer_27, properties {soundWaveAssetPtr=&/Robert_Walk_Vocal_Exhale_15}}
 {id=WavePlayer_28, properties {soundWaveAssetPtr=&/Robert_Walk_Vocal_Exhale_01}}
 {id=WavePlayer_29, properties {soundWaveAssetPtr=&Robert_Walk_Vocal_Exhale_02}}
 {id=WavePlayer_3, properties {soundWaveAssetPtr=&Robert_Walk_Vocal_Exhale_10}}
 {id=WavePlayer_30, properties {soundWaveAssetPtr=&Robert_Walk_Vocal_Exhale_03}}
 {id=WavePlayer_31, properties {soundWaveAssetPtr=&Robert_Walk_Vocal_Exhale_04}}
 {id=WavePlayer_32, properties {soundWaveAssetPtr=&Robert_Walk_Vocal_Exhale_05}}
 {id=WavePlayer_33, properties {soundWaveAssetPtr=&Robert_Walk_Vocal_Exhale_06}}
 {id=WavePlayer_34, properties {soundWaveAssetPtr=&Robert_Walk_Vocal_Exhale_07}}
 {id=WavePlayer_35, properties {soundWaveAssetPtr=&/Robert_Walk_Vocal_Exhale_09}}
 {id=WavePlayer_36, properties {soundWaveAssetPtr=&Robert_Walk_Vocal_Exhale_09}}
 {id=WavePlayer_37, properties {soundWaveAssetPtr=&Robert_Idle_Vocal_Chitter_10}}
 {id=WavePlayer_38, properties {soundWaveAssetPtr=&Robert_Idle_Vocal_Chitter_11}}
 {id=WavePlayer_39, properties {soundWaveAssetPtr=&Robert_Idle_Vocal_Chitter_15}}
 {id=WavePlayer_4, properties {soundWaveAssetPtr=&Robert_Walk_Vocal_Exhale_11}}
 {id=WavePlayer_40, properties {soundWaveAssetPtr=&Robert_Idle_Vocal_Chitter_16}}
 {id=WavePlayer_41, properties {soundWaveAssetPtr=&Robert_Idle_Vocal_Chitter_17}}
 {id=WavePlayer_42, properties {soundWaveAssetPtr=&Robert_Idle_Vocal_Chitter_21}}
 {id=WavePlayer_43, properties {soundWaveAssetPtr=&Robert_Idle_Vocal_Chitter_22}}
 {id=WavePlayer_44, properties {soundWaveAssetPtr=&/Robert_Idle_Vocal_Chitter_01}}
 {id=WavePlayer_45, properties {soundWaveAssetPtr=&/Robert_Idle_Vocal_Chitter_02}}
 {id=WavePlayer_46, properties {soundWaveAssetPtr=&/Robert_Idle_Vocal_Chitter_04}}
 {id=WavePlayer_47, properties {soundWaveAssetPtr=&/Robert_Idle_Vocal_Chitter_06}}
 {id=WavePlayer_48, properties {soundWaveAssetPtr=&/Robert_Idle_Vocal_Chitter_07}}
 {id=WavePlayer_49, properties {soundWaveAssetPtr=&/Robert_Idle_Vocal_Chitter_08}}
 {id=WavePlayer_5, properties {soundWaveAssetPtr=&Robert_Walk_Vocal_Exhale_12}}
 {id=WavePlayer_50, properties {soundWaveAssetPtr=&/Robert_Idle_Vocal_Chitter_09}}
 {id=WavePlayer_51, properties {soundWaveAssetPtr=&Robert_Walk_Vocal_Exhale_10}}
 {id=WavePlayer_52, properties {soundWaveAssetPtr=&Robert_Walk_Vocal_Exhale_11}}
 {id=WavePlayer_53, properties {soundWaveAssetPtr=&Robert_Walk_Vocal_Exhale_12}}
 {id=WavePlayer_54, properties {soundWaveAssetPtr=&Robert_Walk_Vocal_Exhale_13}}
 {id=WavePlayer_55, properties {soundWaveAssetPtr=&Robert_Walk_Vocal_Exhale_14}}
 {id=WavePlayer_56, properties {soundWaveAssetPtr=&Robert_Walk_Vocal_Exhale_15}}
 {id=WavePlayer_57, properties {soundWaveAssetPtr=&Robert_Walk_Vocal_Exhale_01}}
 {id=WavePlayer_58, properties {soundWaveAssetPtr=&Robert_Walk_Vocal_Exhale_02}}
 {id=WavePlayer_59, properties {soundWaveAssetPtr=&Robert_Idle_Vocal_Chitter_01}}
 {id=WavePlayer_6, properties {soundWaveAssetPtr=&Robert_Walk_Vocal_Exhale_13}}
 {id=WavePlayer_60, properties {soundWaveAssetPtr=&Robert_Idle_Vocal_Chitter_02}}
 {id=WavePlayer_61, properties {soundWaveAssetPtr=&Robert_Idle_Vocal_Chitter_04}}
 {id=WavePlayer_62, properties {soundWaveAssetPtr=&Robert_Idle_Vocal_Chitter_06}}
 {id=WavePlayer_63, properties {soundWaveAssetPtr=&Robert_Idle_Vocal_Chitter_07}}
 {id=WavePlayer_64, properties {soundWaveAssetPtr=&Robert_Idle_Vocal_Chitter_08}}
 {id=WavePlayer_65, properties {soundWaveAssetPtr=&Robert_Idle_Vocal_Chitter_09}}
 {id=WavePlayer_66, properties {soundWaveAssetPtr=&Robert_Walk_Vocal_Exhale_03}}
 {id=WavePlayer_67, properties {soundWaveAssetPtr=&Robert_Walk_Vocal_Exhale_04}}
 {id=WavePlayer_68, properties {soundWaveAssetPtr=&Robert_Walk_Vocal_Exhale_05}}
 {id=WavePlayer_69, properties {soundWaveAssetPtr=&Robert_Walk_Vocal_Exhale_06}}
 {id=WavePlayer_7, properties {soundWaveAssetPtr=&Robert_Walk_Vocal_Exhale_14}}
 {id=WavePlayer_70, properties {soundWaveAssetPtr=&Robert_Walk_Vocal_Exhale_07}}
 {id=WavePlayer_71, properties {soundWaveAssetPtr=&Robert_Walk_Vocal_Exhale_08}}
 {id=WavePlayer_72, properties {soundWaveAssetPtr=&Robert_Walk_Vocal_Exhale_09}}
 {id=WavePlayer_73, properties {soundWaveAssetPtr=&Robert_Idle_Vocal_Chitter_10}}
 {id=WavePlayer_74, properties {soundWaveAssetPtr=&Robert_Idle_Vocal_Chitter_11}}
 {id=WavePlayer_75, properties {soundWaveAssetPtr=&Robert_Idle_Vocal_Chitter_15}}
 {id=WavePlayer_76, properties {soundWaveAssetPtr=&Robert_Idle_Vocal_Chitter_16}}
 {id=WavePlayer_77, properties {soundWaveAssetPtr=&Robert_Idle_Vocal_Chitter_17}}
 {id=WavePlayer_78, properties {soundWaveAssetPtr=&Robert_Idle_Vocal_Chitter_21}}
 {id=WavePlayer_79, properties {soundWaveAssetPtr=&Robert_Idle_Vocal_Chitter_22}}
 {id=WavePlayer_8, properties {soundWaveAssetPtr=&Robert_Walk_Vocal_Exhale_15}}
 {id=WavePlayer_80, properties {soundWaveAssetPtr=&Robert_Idle_Vocal_Chitter_01}}
 {id=WavePlayer_81, properties {soundWaveAssetPtr=&Robert_Idle_Vocal_Chitter_02}}
 {id=WavePlayer_82, properties {soundWaveAssetPtr=&Robert_Idle_Vocal_Chitter_04}}
 {id=WavePlayer_83, properties {soundWaveAssetPtr=&Robert_Idle_Vocal_Chitter_06}}
 {id=WavePlayer_84, properties {soundWaveAssetPtr=&Robert_Idle_Vocal_Chitter_07}}
 {id=WavePlayer_85, properties {soundWaveAssetPtr=&Robert_Idle_Vocal_Chitter_08}}
 {id=WavePlayer_86, properties {soundWaveAssetPtr=&Robert_Idle_Vocal_Chitter_09}}
 {id=WavePlayer_9, properties {soundWaveAssetPtr=&Robert_Walk_Vocal_Exhale_01}}
]
properties {
 maxDistance=2097152
 priority=0.98
 soundClassObject=&Fort_Creatures_Vocalization_Tactical
 volumeMultiplier=1.25
 concurrencySet [/Irwin/Sounds/Robert/Robert_Vocal_Breathing_Concurrency_Owner, /Irwin/Sounds/Robert/Robert_Vocal_Breathing_Concurrency_Global]
}
