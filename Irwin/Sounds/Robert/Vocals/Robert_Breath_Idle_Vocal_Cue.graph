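# Sound cue for the creature Robert's idle breathing vocals, limited by per-owner and global breathing concurrency.
# IsLocal selects a moderately attenuated-volume mix; otherwise IsRiddenByEnemy picks full volume when true or unset, and a much quieter mix when false.
# Each path feeds the same predator attenuation and weighted random pool: mostly pitch- and volume-modulated walk exhales, sometimes idle chitters or empty slots for silence.
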SoundCue /Irwin/Sounds/Robert/Vocals/Robert_Breath_Idle_Vocal_Cue
root=Branch_1
nodes [
 {id=Attenuation_0, children [Random_9], properties {attenuationSettings=/Irwin/Sounds/Settings/Predator_Vocalization_Medium_Att}}
 {id=Attenuation_6, children [Random_15], properties {attenuationSettings=&Predator_Vocalization_Medium_Att}}
 {id=Attenuation_8, children [Random_21], properties {attenuationSettings=&Predator_Vocalization_Medium_Att}}
 {id=Branch_0, children [Mixer_2,Mixer_1,Mixer_2], properties {boolParameterName=IsRiddenByEnemy}}
 {id=Branch_1, children [Mixer_3,Branch_0,Branch_0], properties {boolParameterName=IsLocal}}
 {id=Mixer_1, children [Attenuation_8], properties {inputVolume [0.3]}}
 {id=Mixer_2, children [Attenuation_6], properties {inputVolume [1]}}
 {id=Mixer_3, children [Attenuation_0], properties {inputVolume [0.7]}}
 {id=Modulator_2, children [Random_17], properties {pitchMax=1.2, pitchMin=1, volumeMin=0.1}}
 {id=Modulator_4, children [Random_11], properties {pitchMax=1.2, pitchMin=1, volumeMin=0.1}}
 {id=Modulator_5, children [Random_23], properties {pitchMax=1.2, pitchMin=1, volumeMin=0.1}}
 {
  id=Random_10
  children [
   WavePlayer_66
   WavePlayer_67
   WavePlayer_68
   WavePlayer_69
   WavePlayer_70
   WavePlayer_71
   WavePlayer_72
   WavePlayer_73
   WavePlayer_74
   WavePlayer_75
   WavePlayer_76
   WavePlayer_77
   WavePlayer_78
   WavePlayer_79
  ]
  properties {bShouldExcludeFromBranchCulling=true, weights [1,1,1,1,1,1,1,1,1,1,1,1,1,1]}
 }
 {
  id=Random_11
  children [
   WavePlayer_100
   WavePlayer_101
   WavePlayer_102
   WavePlayer_103
   WavePlayer_104
   WavePlayer_105
   WavePlayer_106
   WavePlayer_107
   WavePlayer_108
   WavePlayer_109
   WavePlayer_110
   WavePlayer_111
   WavePlayer_112
   WavePlayer_113
   WavePlayer_114
  ]
  properties {bShouldExcludeFromBranchCulling=true, weights [1,1,1,1,1,1,1,1,1,1,1,1,1,1,1]}
 }
 {
  id=Random_15
  children [Modulator_2,Modulator_2,Modulator_2,Modulator_2,Modulator_2,Modulator_2,Random_16,null,null,null]
  properties {bShouldExcludeFromBranchCulling=true, weights [1,1,1,3,1,1,1,1,1,1]}
 }
 {
  id=Random_16
  children [
   WavePlayer_151
   WavePlayer_152
   WavePlayer_153
   WavePlayer_154
   WavePlayer_155
   WavePlayer_156
   WavePlayer_157
   WavePlayer_158
   WavePlayer_161
   WavePlayer_162
   WavePlayer_163
   WavePlayer_164
   WavePlayer_165
   WavePlayer_166
  ]
  properties {bShouldExcludeFromBranchCulling=true, weights [1,1,1,1,1,1,1,1,1,1,1,1,1,1]}
 }
 {
  id=Random_17
  children [
   WavePlayer_167
   WavePlayer_168
   WavePlayer_169
   WavePlayer_170
   WavePlayer_171
   WavePlayer_172
   WavePlayer_173
   WavePlayer_174
   WavePlayer_175
   WavePlayer_176
   WavePlayer_177
   WavePlayer_178
   WavePlayer_179
   WavePlayer_180
   WavePlayer_181
  ]
  properties {bShouldExcludeFromBranchCulling=true, weights [1,1,1,1,1,1,1,1,1,1,1,1,1,1,1]}
 }
 {
  id=Random_21
  children [Modulator_5,Modulator_5,Modulator_5,Modulator_5,Modulator_5,Modulator_5,Random_22,null,null,null]
  properties {bShouldExcludeFromBranchCulling=true, weights [1,1,1,3,1,1,1,1,1,1]}
 }
 {
  id=Random_22
  children [
   WavePlayer_211
   WavePlayer_212
   WavePlayer_213
   WavePlayer_214
   WavePlayer_215
   WavePlayer_216
   WavePlayer_217
   WavePlayer_218
   WavePlayer_219
   WavePlayer_220
   WavePlayer_221
   WavePlayer_222
   WavePlayer_223
   WavePlayer_224
  ]
  properties {bShouldExcludeFromBranchCulling=true, weights [1,1,1,1,1,1,1,1,1,1,1,1,1,1]}
 }
 {
  id=Random_23
  children [
   WavePlayer_225
   WavePlayer_226
   WavePlayer_227
   WavePlayer_228
   WavePlayer_229
   WavePlayer_230
   WavePlayer_231
   WavePlayer_232
   WavePlayer_233
   WavePlayer_234
   WavePlayer_235
   WavePlayer_236
   WavePlayer_237
   WavePlayer_238
   WavePlayer_239
  ]
  properties {bShouldExcludeFromBranchCulling=true, weights [1,1,1,1,1,1,1,1,1,1,1,1,1,1,1]}
 }
 {
  id=Random_9
  children [Modulator_4,Modulator_4,Modulator_4,Modulator_4,Modulator_4,Modulator_4,Random_10,null,null,null]
  properties {bShouldExcludeFromBranchCulling=true, weights [1,1,1,3,1,1,1,1,1,1]}
 }
 {id=WavePlayer_100, properties {soundWaveAssetPtr=&/Robert_Walk_Vocal_Exhale_10}}
 {id=WavePlayer_101, properties {soundWaveAssetPtr=&/Robert_Walk_Vocal_Exhale_11}}
 {id=WavePlayer_102, properties {soundWaveAssetPtr=&/Robert_Walk_Vocal_Exhale_12}}
 {id=WavePlayer_103, properties {soundWaveAssetPtr=&/Robert_Walk_Vocal_Exhale_13}}
 {id=WavePlayer_104, properties {soundWaveAssetPtr=&/Robert_Walk_Vocal_Exhale_14}}
 {id=WavePlayer_105, properties {soundWaveAssetPtr=&/Robert_Walk_Vocal_Exhale_15}}
 {id=WavePlayer_106, properties {soundWaveAssetPtr=&/Robert_Walk_Vocal_Exhale_01}}
 {id=WavePlayer_107, properties {soundWaveAssetPtr=&/Robert_Walk_Vocal_Exhale_02}}
 {id=WavePlayer_108, properties {soundWaveAssetPtr=&/Robert_Walk_Vocal_Exhale_03}}
 {id=WavePlayer_109, properties {soundWaveAssetPtr=&/Robert_Walk_Vocal_Exhale_04}}
 {id=WavePlayer_110, properties {soundWaveAssetPtr=&/Robert_Walk_Vocal_Exhale_05}}
 {id=WavePlayer_111, properties {soundWaveAssetPtr=&/Robert_Walk_Vocal_Exhale_06}}
 {id=WavePlayer_112, properties {soundWaveAssetPtr=&/Robert_Walk_Vocal_Exhale_07}}
 {id=WavePlayer_113, properties {soundWaveAssetPtr=&/Robert_Walk_Vocal_Exhale_08}}
 {id=WavePlayer_114, properties {soundWaveAssetPtr=&/Robert_Walk_Vocal_Exhale_09}}
 {id=WavePlayer_151, properties {soundWaveAssetPtr=&/Robert_Idle_Vocal_Chitter_10}}
 {id=WavePlayer_152, properties {soundWaveAssetPtr=&/Robert_Idle_Vocal_Chitter_11}}
 {id=WavePlayer_153, properties {soundWaveAssetPtr=&/Robert_Idle_Vocal_Chitter_15}}
 {id=WavePlayer_154, properties {soundWaveAssetPtr=&/Robert_Idle_Vocal_Chitter_16}}
 {id=WavePlayer_155, properties {soundWaveAssetPtr=&/Robert_Idle_Vocal_Chitter_17}}
 {id=WavePlayer_156, properties {soundWaveAssetPtr=&/Robert_Idle_Vocal_Chitter_21}}
 {id=WavePlayer_157, properties {soundWaveAssetPtr=&/Robert_Idle_Vocal_Chitter_22}}
 {id=WavePlayer_158, properties {soundWaveAssetPtr=&/Robert_Idle_Vocal_Chitter_01}}
 {id=WavePlayer_161, properties {soundWaveAssetPtr=&/Robert_Idle_Vocal_Chitter_02}}
 {id=WavePlayer_162, properties {soundWaveAssetPtr=&/Robert_Idle_Vocal_Chitter_04}}
 {id=WavePlayer_163, properties {soundWaveAssetPtr=&/Robert_Idle_Vocal_Chitter_06}}
 {id=WavePlayer_164, properties {soundWaveAssetPtr=&/Robert_Idle_Vocal_Chitter_07}}
 {id=WavePlayer_165, properties {soundWaveAssetPtr=&/Robert_Idle_Vocal_Chitter_08}}
 {id=WavePlayer_166, properties {soundWaveAssetPtr=&/Robert_Idle_Vocal_Chitter_09}}
 {id=WavePlayer_167, properties {soundWaveAssetPtr=&Robert_Walk_Vocal_Exhale_10}}
 {id=WavePlayer_168, properties {soundWaveAssetPtr=&Robert_Walk_Vocal_Exhale_11}}
 {id=WavePlayer_169, properties {soundWaveAssetPtr=&Robert_Walk_Vocal_Exhale_12}}
 {id=WavePlayer_170, properties {soundWaveAssetPtr=&Robert_Walk_Vocal_Exhale_13}}
 {id=WavePlayer_171, properties {soundWaveAssetPtr=&Robert_Walk_Vocal_Exhale_14}}
 {id=WavePlayer_172, properties {soundWaveAssetPtr=&Robert_Walk_Vocal_Exhale_15}}
 {id=WavePlayer_173, properties {soundWaveAssetPtr=&Robert_Walk_Vocal_Exhale_01}}
 {id=WavePlayer_174, properties {soundWaveAssetPtr=&Robert_Walk_Vocal_Exhale_02}}
 {id=WavePlayer_175, properties {soundWaveAssetPtr=&Robert_Walk_Vocal_Exhale_03}}
 {id=WavePlayer_176, properties {soundWaveAssetPtr=&Robert_Walk_Vocal_Exhale_04}}
 {id=WavePlayer_177, properties {soundWaveAssetPtr=&Robert_Walk_Vocal_Exhale_05}}
 {id=WavePlayer_178, properties {soundWaveAssetPtr=&Robert_Walk_Vocal_Exhale_06}}
 {id=WavePlayer_179, properties {soundWaveAssetPtr=&Robert_Walk_Vocal_Exhale_07}}
 {id=WavePlayer_180, properties {soundWaveAssetPtr=&Robert_Walk_Vocal_Exhale_08}}
 {id=WavePlayer_181, properties {soundWaveAssetPtr=&Robert_Walk_Vocal_Exhale_09}}
 {id=WavePlayer_211, properties {soundWaveAssetPtr=&Robert_Idle_Vocal_Chitter_10}}
 {id=WavePlayer_212, properties {soundWaveAssetPtr=&Robert_Idle_Vocal_Chitter_11}}
 {id=WavePlayer_213, properties {soundWaveAssetPtr=&Robert_Idle_Vocal_Chitter_15}}
 {id=WavePlayer_214, properties {soundWaveAssetPtr=&Robert_Idle_Vocal_Chitter_16}}
 {id=WavePlayer_215, properties {soundWaveAssetPtr=&Robert_Idle_Vocal_Chitter_17}}
 {id=WavePlayer_216, properties {soundWaveAssetPtr=&Robert_Idle_Vocal_Chitter_21}}
 {id=WavePlayer_217, properties {soundWaveAssetPtr=&Robert_Idle_Vocal_Chitter_22}}
 {id=WavePlayer_218, properties {soundWaveAssetPtr=&Robert_Idle_Vocal_Chitter_01}}
 {id=WavePlayer_219, properties {soundWaveAssetPtr=&Robert_Idle_Vocal_Chitter_02}}
 {id=WavePlayer_220, properties {soundWaveAssetPtr=&Robert_Idle_Vocal_Chitter_04}}
 {id=WavePlayer_221, properties {soundWaveAssetPtr=&Robert_Idle_Vocal_Chitter_06}}
 {id=WavePlayer_222, properties {soundWaveAssetPtr=&Robert_Idle_Vocal_Chitter_07}}
 {id=WavePlayer_223, properties {soundWaveAssetPtr=&Robert_Idle_Vocal_Chitter_08}}
 {id=WavePlayer_224, properties {soundWaveAssetPtr=&Robert_Idle_Vocal_Chitter_09}}
 {id=WavePlayer_225, properties {soundWaveAssetPtr=&Robert_Walk_Vocal_Exhale_10}}
 {id=WavePlayer_226, properties {soundWaveAssetPtr=&Robert_Walk_Vocal_Exhale_11}}
 {id=WavePlayer_227, properties {soundWaveAssetPtr=&Robert_Walk_Vocal_Exhale_12}}
 {id=WavePlayer_228, properties {soundWaveAssetPtr=&Robert_Walk_Vocal_Exhale_13}}
 {id=WavePlayer_229, properties {soundWaveAssetPtr=&Robert_Walk_Vocal_Exhale_14}}
 {id=WavePlayer_230, properties {soundWaveAssetPtr=&Robert_Walk_Vocal_Exhale_15}}
 {id=WavePlayer_231, properties {soundWaveAssetPtr=&Robert_Walk_Vocal_Exhale_01}}
 {id=WavePlayer_232, properties {soundWaveAssetPtr=&Robert_Walk_Vocal_Exhale_02}}
 {id=WavePlayer_233, properties {soundWaveAssetPtr=&Robert_Walk_Vocal_Exhale_03}}
 {id=WavePlayer_234, properties {soundWaveAssetPtr=&Robert_Walk_Vocal_Exhale_04}}
 {id=WavePlayer_235, properties {soundWaveAssetPtr=&Robert_Walk_Vocal_Exhale_05}}
 {id=WavePlayer_236, properties {soundWaveAssetPtr=&Robert_Walk_Vocal_Exhale_06}}
 {id=WavePlayer_237, properties {soundWaveAssetPtr=&Robert_Walk_Vocal_Exhale_07}}
 {id=WavePlayer_238, properties {soundWaveAssetPtr=&Robert_Walk_Vocal_Exhale_08}}
 {id=WavePlayer_239, properties {soundWaveAssetPtr=&Robert_Walk_Vocal_Exhale_09}}
 {id=WavePlayer_66, properties {soundWaveAssetPtr=&Robert_Idle_Vocal_Chitter_10}}
 {id=WavePlayer_67, properties {soundWaveAssetPtr=&Robert_Idle_Vocal_Chitter_11}}
 {id=WavePlayer_68, properties {soundWaveAssetPtr=&Robert_Idle_Vocal_Chitter_15}}
 {id=WavePlayer_69, properties {soundWaveAssetPtr=&Robert_Idle_Vocal_Chitter_16}}
 {id=WavePlayer_70, properties {soundWaveAssetPtr=&Robert_Idle_Vocal_Chitter_17}}
 {id=WavePlayer_71, properties {soundWaveAssetPtr=&Robert_Idle_Vocal_Chitter_21}}
 {id=WavePlayer_72, properties {soundWaveAssetPtr=&Robert_Idle_Vocal_Chitter_22}}
 {id=WavePlayer_73, properties {soundWaveAssetPtr=&Robert_Idle_Vocal_Chitter_01}}
 {id=WavePlayer_74, properties {soundWaveAssetPtr=&Robert_Idle_Vocal_Chitter_02}}
 {id=WavePlayer_75, properties {soundWaveAssetPtr=&Robert_Idle_Vocal_Chitter_04}}
 {id=WavePlayer_76, properties {soundWaveAssetPtr=&Robert_Idle_Vocal_Chitter_06}}
 {id=WavePlayer_77, properties {soundWaveAssetPtr=&Robert_Idle_Vocal_Chitter_07}}
 {id=WavePlayer_78, properties {soundWaveAssetPtr=&Robert_Idle_Vocal_Chitter_08}}
 {id=WavePlayer_79, properties {soundWaveAssetPtr=&Robert_Idle_Vocal_Chitter_09}}
]
properties {
 attenuationSettings=&Predator_Vocalization_Medium_Att
 maxDistance=12000
 priority=0.95
 soundClassObject=/Game/Sounds/FortSoundClassesAndModes/Fort_Creatures_Vocalization_Normal
 volumeMultiplier=1.25
 concurrencySet [/Irwin/Sounds/Robert/Robert_Vocal_Breathing_Concurrency_Owner, /Irwin/Sounds/Robert/Robert_Vocal_Breathing_Concurrency_Global]
}
