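# Sound cue for Robert's breathing vocal on a 180-degree swim turn, using creature vocalization class, predator attenuation and owner/global breathing concurrency.
# Branches silence it when IsLocal is true or IsRiddenByEnemy is false; unset parameters let it play.
# The audible path runs through a mixer into an equal-weight random pick of twenty turn vocal waves.
# The random node is excluded from branch culling.
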SoundCue /Irwin/Sounds/Robert/Vocals/Robert_Breath_Idle_Vocal_Turn_Swim_180_Cue
root=Branch_1
nodes [
 {id=Branch_0, children [Mixer_2,null,Mixer_2], properties {boolParameterName=IsRiddenByEnemy}}
 {id=Branch_1, children [null,Branch_0,Branch_0], properties {boolParameterName=IsLocal}}
 {id=Mixer_2, children [Random_0], properties {inputVolume [1]}}
 {
  id=Random_0
  children [
   WavePlayer_0
   WavePlayer_1
   WavePlayer_2
   WavePlayer_3
   WavePlayer_4
   WavePlayer_5
   WavePlayer_6
   WavePlayer_7
   WavePlayer_8
   WavePlayer_9
   WavePlayer_10
   WavePlayer_11
   WavePlayer_12
   WavePlayer_13
   WavePlayer_14
   WavePlayer_15
   WavePlayer_16
   WavePlayer_17
   WavePlayer_18
   WavePlayer_19
  ]
  properties {bShouldExcludeFromBranchCulling=true, weights [1,1,1,1,1,1,1,1,1,1,1,1,1,1,1,1,1,1,1,1]}
 }
 {id=WavePlayer_0, properties {soundWaveAssetPtr=&/Robert_Turn_180_Vocal_10}}
 {id=WavePlayer_1, properties {soundWaveAssetPtr=&/Robert_Turn_180_Vocal_11}}
 {id=WavePlayer_10, properties {soundWaveAssetPtr=&/Robert_Turn_180_Vocal_29}}
 {id=WavePlayer_11, properties {soundWaveAssetPtr=&/Robert_Turn_180_Vocal_30}}
 {id=WavePlayer_12, properties {soundWaveAssetPtr=&/Robert_Turn_180_Vocal_01}}
 {id=WavePlayer_13, properties {soundWaveAssetPtr=&/Robert_Turn_180_Vocal_02}}
 {id=WavePlayer_14, properties {soundWaveAssetPtr=&/Robert_Turn_180_Vocal_03}}
 {id=WavePlayer_15, properties {soundWaveAssetPtr=&/Robert_Turn_180_Vocal_04}}
 {id=WavePlayer_16, properties {soundWaveAssetPtr=&/Robert_Turn_180_Vocal_05}}
 {id=WavePlayer_17, properties {soundWaveAssetPtr=&/Robert_Turn_180_Vocal_06}}
 {id=WavePlayer_18, properties {soundWaveAssetPtr=&/Robert_Turn_180_Vocal_07}}
 {id=WavePlayer_19, properties {soundWaveAssetPtr=&/Robert_Turn_180_Vocal_08}}
 {id=WavePlayer_2, properties {soundWaveAssetPtr=&/Robert_Turn_180_Vocal_12}}
 {id=WavePlayer_3, properties {soundWaveAssetPtr=&/Robert_Turn_180_Vocal_17}}
 {id=WavePlayer_4, properties {soundWaveAssetPtr=&/Robert_Turn_180_Vocal_18}}
 {id=WavePlayer_5, properties {soundWaveAssetPtr=&/Robert_Turn_180_Vocal_21}}
 {id=WavePlayer_6, properties {soundWaveAssetPtr=&/Robert_Turn_180_Vocal_22}}
 {id=WavePlayer_7, properties {soundWaveAssetPtr=&/Robert_Turn_180_Vocal_24}}
 {id=WavePlayer_8, properties {soundWaveAssetPtr=&/Robert_Turn_180_Vocal_25}}
 {id=WavePlayer_9, properties {soundWaveAssetPtr=&/Robert_Turn_180_Vocal_26}}
]
properties {
 attenuationSettings=/Irwin/Sounds/Settings/Predator_Vocalization_Medium_Att
 maxDistance=12000
 priority=0.95
 soundClassObject=/Game/Sounds/FortSoundClassesAndModes/Fort_Creatures_Vocalization_Normal
 volumeMultiplier=1.25
 concurrencySet [/Irwin/Sounds/Robert/Robert_Vocal_Breathing_Concurrency_Owner, /Irwin/Sounds/Robert/Robert_Vocal_Breathing_Concurrency_Global]
}
