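# Sound cue for the creature Robert's running breath vocals, mostly exhales with occasional chitters or silence.
# IsLocal and IsRiddenByEnemy branches set loudness: medium for local, full for enemy-ridden or unset, quiet for other non-local cases.
# Each path duplicates the same weighted random tree through predator vocalization attenuation, varying pitch and volume on exhales only.
# Owner and global concurrency sets govern overlapping breaths.
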SoundCue /Irwin/Sounds/Robert/Vocals/Robert_Breath_Run_Vocal_Cue
root=Branch_3
nodes [
 {id=Attenuation_6, children [Random_15], properties {attenuationSettings=/Irwin/Sounds/Settings/Predator_Vocalization_Medium_Att}}
 {id=Attenuation_7, children [Random_18], properties {attenuationSettings=&Predator_Vocalization_Medium_Att}}
 {id=Attenuation_8, children [Random_21], properties {attenuationSettings=&Predator_Vocalization_Medium_Att}}
 {id=Branch_1, children [Mixer_7,Mixer_6,Mixer_7], properties {boolParameterName=IsRiddenByEnemy}}
 {id=Branch_3, children [Mixer_8,Branch_1,Branch_1], properties {boolParameterName=IsLocal}}
 {id=Mixer_6, children [Attenuation_8], properties {inputVolume [0.3]}}
 {id=Mixer_7, children [Attenuation_7], properties {inputVolume [1]}}
 {id=Mixer_8, children [Attenuation_6], properties {inputVolume [0.7]}}
 {id=Modulator_3, children [Random_20], properties {pitchMax=1.2, pitchMin=1, volumeMin=0.1}}
 {id=Modulator_4, children [Random_17], properties {pitchMax=1.2, pitchMin=1, volumeMin=0.1}}
 {id=Modulator_5, children [Random_23], properties {pitchMax=1.2, pitchMin=1, volumeMin=0.1}}
 {
  id=Random_15
  children [Modulator_4,Modulator_4,Modulator_4,Modulator_4,Modulator_4,Modulator_4,Random_16,null,null,null]
  properties {bShouldExcludeFromBranchCulling=true, weights [1,1,1,3,1,1,1,1,1,1]}
 }
 {
  id=Random_16
  children [
   WavePlayer_138
   WavePlayer_139
   WavePlayer_140
   WavePlayer_141
   WavePlayer_142
   WavePlayer_143
   WavePlayer_144
   WavePlayer_145
   WavePlayer_146
   WavePlayer_147
   WavePlayer_148
   WavePlayer_149
   WavePlayer_150
   WavePlayer_151
  ]
  properties {bShouldExcludeFromBranchCulling=true, weights [1,1,1,1,1,1,1,1,1,1,1,1,1,1]}
 }
 {
  id=Random_17
  children [
   WavePlayer_152
   WavePlayer_153
   WavePlayer_154
   WavePlayer_155
   WavePlayer_156
   WavePlayer_157
   WavePlayer_161
   WavePlayer_162
   WavePlayer_163
   WavePlayer_164
   WavePlayer_165
   WavePlayer_166
   WavePlayer_167
   WavePlayer_168
   WavePlayer_169
  ]
  properties {bShouldExcludeFromBranchCulling=true, weights [1,1,1,1,1,1,1,1,1,1,1,1,1,1,1]}
 }
 {
  id=Random_18
  children [Modulator_3,Modulator_3,Modulator_3,Modulator_3,Modulator_3,Modulator_3,Random_19,null,null,null]
  properties {bShouldExcludeFromBranchCulling=true, weights [1,1,1,3,1,1,1,1,1,1]}
 }
 {
  id=Random_19
  children [
   WavePlayer_170
   WavePlayer_171
   WavePlayer_172
   WavePlayer_173
   WavePlayer_174
   WavePlayer_175
   WavePlayer_176
   WavePlayer_177
   WavePlayer_178
   WavePlayer_179
   WavePlayer_180
   WavePlayer_181
   WavePlayer_182
   WavePlayer_183
  ]
  properties {bShouldExcludeFromBranchCulling=true, weights [1,1,1,1,1,1,1,1,1,1,1,1,1,1]}
 }
 {
  id=Random_20
  children [
   WavePlayer_184
   WavePlayer_185
   WavePlayer_186
   WavePlayer_187
   WavePlayer_188
   WavePlayer_189
   WavePlayer_190
   WavePlayer_191
   WavePlayer_192
   WavePlayer_193
   WavePlayer_194
   WavePlayer_195
   WavePlayer_196
   WavePlayer_197
   WavePlayer_198
  ]
  properties {bShouldExcludeFromBranchCulling=true, weights [1,1,1,1,1,1,1,1,1,1,1,1,1,1,1]}
 }
 {
  id=Random_21
  children [Modulator_5,Modulator_5,Modulator_5,Modulator_5,Modulator_5,Modulator_5,Random_22,null,null,null]
  properties {bShouldExcludeFromBranchCulling=true, weights [1,1,1,3,1,1,1,1,1,1]}
 }
 {
  id=Random_22
  children [
   WavePlayer_211
   WavePlayer_212
   WavePlayer_213
   WavePlayer_214
   WavePlayer_215
   WavePlayer_216
   WavePlayer_217
   WavePlayer_218
   WavePlayer_219
   WavePlayer_220
   WavePlayer_221
   WavePlayer_222
   WavePlayer_223
   WavePlayer_224
  ]
  properties {bShouldExcludeFromBranchCulling=true, weights [1,1,1,1,1,1,1,1,1,1,1,1,1,1]}
 }
 {
  id=Random_23
  children [
   WavePlayer_225
   WavePlayer_226
   WavePlayer_227
   WavePlayer_228
   WavePlayer_229
   WavePlayer_230
   WavePlayer_231
   WavePlayer_232
   WavePlayer_233
   WavePlayer_234
   WavePlayer_235
   WavePlayer_236
   WavePlayer_237
   WavePlayer_238
   WavePlayer_239
  ]
  properties {bShouldExcludeFromBranchCulling=true, weights [1,1,1,1,1,1,1,1,1,1,1,1,1,1,1]}
 }
 {id=WavePlayer_138, properties {soundWaveAssetPtr=&/Robert_Idle_Vocal_Chitter_10}}
 {id=WavePlayer_139, properties {soundWaveAssetPtr=&/Robert_Idle_Vocal_Chitter_11}}
 {id=WavePlayer_140, properties {soundWaveAssetPtr=&/Robert_Idle_Vocal_Chitter_15}}
 {id=WavePlayer_141, properties {soundWaveAssetPtr=&/Robert_Idle_Vocal_Chitter_16}}
 {id=WavePlayer_142, properties {soundWaveAssetPtr=&/Robert_Idle_Vocal_Chitter_17}}
 {id=WavePlayer_143, properties {soundWaveAssetPtr=&/Robert_Idle_Vocal_Chitter_21}}
 {id=WavePlayer_144, properties {soundWaveAssetPtr=&/Robert_Idle_Vocal_Chitter_22}}
 {id=WavePlayer_145, properties {soundWaveAssetPtr=&/Robert_Idle_Vocal_Chitter_01}}
 {id=WavePlayer_146, properties {soundWaveAssetPtr=&/Robert_Idle_Vocal_Chitter_02}}
 {id=WavePlayer_147, properties {soundWaveAssetPtr=&/Robert_Idle_Vocal_Chitter_04}}
 {id=WavePlayer_148, properties {soundWaveAssetPtr=&/Robert_Idle_Vocal_Chitter_06}}
 {id=WavePlayer_149, properties {soundWaveAssetPtr=&/Robert_Idle_Vocal_Chitter_07}}
 {id=WavePlayer_150, properties {soundWaveAssetPtr=&/Robert_Idle_Vocal_Chitter_08}}
 {id=WavePlayer_151, properties {soundWaveAssetPtr=&/Robert_Idle_Vocal_Chitter_09}}
 {id=WavePlayer_152, properties {soundWaveAssetPtr=&/Robert_Walk_Vocal_Exhale_10}}
 {id=WavePlayer_153, properties {soundWaveAssetPtr=&/Robert_Walk_Vocal_Exhale_11}}
 {id=WavePlayer_154, properties {soundWaveAssetPtr=&/Robert_Walk_Vocal_Exhale_12}}
 {id=WavePlayer_155, properties {soundWaveAssetPtr=&/Robert_Walk_Vocal_Exhale_13}}
 {id=WavePlayer_156, properties {soundWaveAssetPtr=&/Robert_Walk_Vocal_Exhale_14}}
 {id=WavePlayer_157, properties {soundWaveAssetPtr=&/Robert_Walk_Vocal_Exhale_15}}
 {id=WavePlayer_161, properties {soundWaveAssetPtr=&/Robert_Walk_Vocal_Exhale_01}}
 {id=WavePlayer_162, properties {soundWaveAssetPtr=&/Robert_Walk_Vocal_Exhale_02}}
 {id=WavePlayer_163, properties {soundWaveAssetPtr=&/Robert_Walk_Vocal_Exhale_03}}
 {id=WavePlayer_164, properties {soundWaveAssetPtr=&/Robert_Walk_Vocal_Exhale_04}}
 {id=WavePlayer_165, properties {soundWaveAssetPtr=&/Robert_Walk_Vocal_Exhale_05}}
 {id=WavePlayer_166, properties {soundWaveAssetPtr=&/Robert_Walk_Vocal_Exhale_06}}
 {id=WavePlayer_167, properties {soundWaveAssetPtr=&/Robert_Walk_Vocal_Exhale_07}}
 {id=WavePlayer_168, properties {soundWaveAssetPtr=&/Robert_Walk_Vocal_Exhale_08}}
 {id=WavePlayer_169, properties {soundWaveAssetPtr=&/Robert_Walk_Vocal_Exhale_09}}
 {id=WavePlayer_170, properties {soundWaveAssetPtr=&Robert_Idle_Vocal_Chitter_10}}
 {id=WavePlayer_171, properties {soundWaveAssetPtr=&Robert_Idle_Vocal_Chitter_11}}
 {id=WavePlayer_172, properties {soundWaveAssetPtr=&Robert_Idle_Vocal_Chitter_15}}
 {id=WavePlayer_173, properties {soundWaveAssetPtr=&Robert_Idle_Vocal_Chitter_16}}
 {id=WavePlayer_174, properties {soundWaveAssetPtr=&Robert_Idle_Vocal_Chitter_17}}
 {id=WavePlayer_175, properties {soundWaveAssetPtr=&Robert_Idle_Vocal_Chitter_21}}
 {id=WavePlayer_176, properties {soundWaveAssetPtr=&Robert_Idle_Vocal_Chitter_22}}
 {id=WavePlayer_177, properties {soundWaveAssetPtr=&Robert_Idle_Vocal_Chitter_01}}
 {id=WavePlayer_178, properties {soundWaveAssetPtr=&Robert_Idle_Vocal_Chitter_02}}
 {id=WavePlayer_179, properties {soundWaveAssetPtr=&Robert_Idle_Vocal_Chitter_04}}
 {id=WavePlayer_180, properties {soundWaveAssetPtr=&Robert_Idle_Vocal_Chitter_06}}
 {id=WavePlayer_181, properties {soundWaveAssetPtr=&Robert_Idle_Vocal_Chitter_07}}
 {id=WavePlayer_182, properties {soundWaveAssetPtr=&Robert_Idle_Vocal_Chitter_08}}
 {id=WavePlayer_183, properties {soundWaveAssetPtr=&Robert_Idle_Vocal_Chitter_09}}
 {id=WavePlayer_184, properties {soundWaveAssetPtr=&Robert_Walk_Vocal_Exhale_10}}
 {id=WavePlayer_185, properties {soundWaveAssetPtr=&Robert_Walk_Vocal_Exhale_11}}
 {id=WavePlayer_186, properties {soundWaveAssetPtr=&Robert_Walk_Vocal_Exhale_12}}
 {id=WavePlayer_187, properties {soundWaveAssetPtr=&Robert_Walk_Vocal_Exhale_13}}
 {id=WavePlayer_188, properties {soundWaveAssetPtr=&Robert_Walk_Vocal_Exhale_14}}
 {id=WavePlayer_189, properties {soundWaveAssetPtr=&Robert_Walk_Vocal_Exhale_15}}
 {id=WavePlayer_190, properties {soundWaveAssetPtr=&Robert_Walk_Vocal_Exhale_01}}
 {id=WavePlayer_191, properties {soundWaveAssetPtr=&Robert_Walk_Vocal_Exhale_02}}
 {id=WavePlayer_192, properties {soundWaveAssetPtr=&Robert_Walk_Vocal_Exhale_03}}
 {id=WavePlayer_193, properties {soundWaveAssetPtr=&Robert_Walk_Vocal_Exhale_04}}
 {id=WavePlayer_194, properties {soundWaveAssetPtr=&Robert_Walk_Vocal_Exhale_05}}
 {id=WavePlayer_195, properties {soundWaveAssetPtr=&Robert_Walk_Vocal_Exhale_06}}
 {id=WavePlayer_196, properties {soundWaveAssetPtr=&Robert_Walk_Vocal_Exhale_07}}
 {id=WavePlayer_197, properties {soundWaveAssetPtr=&Robert_Walk_Vocal_Exhale_08}}
 {id=WavePlayer_198, properties {soundWaveAssetPtr=&Robert_Walk_Vocal_Exhale_09}}
 {id=WavePlayer_211, properties {soundWaveAssetPtr=&Robert_Idle_Vocal_Chitter_10}}
 {id=WavePlayer_212, properties {soundWaveAssetPtr=&Robert_Idle_Vocal_Chitter_11}}
 {id=WavePlayer_213, properties {soundWaveAssetPtr=&Robert_Idle_Vocal_Chitter_15}}
 {id=WavePlayer_214, properties {soundWaveAssetPtr=&Robert_Idle_Vocal_Chitter_16}}
 {id=WavePlayer_215, properties {soundWaveAssetPtr=&Robert_Idle_Vocal_Chitter_17}}
 {id=WavePlayer_216, properties {soundWaveAssetPtr=&Robert_Idle_Vocal_Chitter_21}}
 {id=WavePlayer_217, properties {soundWaveAssetPtr=&Robert_Idle_Vocal_Chitter_22}}
 {id=WavePlayer_218, properties {soundWaveAssetPtr=&Robert_Idle_Vocal_Chitter_01}}
 {id=WavePlayer_219, properties {soundWaveAssetPtr=&Robert_Idle_Vocal_Chitter_02}}
 {id=WavePlayer_220, properties {soundWaveAssetPtr=&Robert_Idle_Vocal_Chitter_04}}
 {id=WavePlayer_221, properties {soundWaveAssetPtr=&Robert_Idle_Vocal_Chitter_06}}
 {id=WavePlayer_222, properties {soundWaveAssetPtr=&Robert_Idle_Vocal_Chitter_07}}
 {id=WavePlayer_223, properties {soundWaveAssetPtr=&Robert_Idle_Vocal_Chitter_08}}
 {id=WavePlayer_224, properties {soundWaveAssetPtr=&Robert_Idle_Vocal_Chitter_09}}
 {id=WavePlayer_225, properties {soundWaveAssetPtr=&Robert_Walk_Vocal_Exhale_10}}
 {id=WavePlayer_226, properties {soundWaveAssetPtr=&Robert_Walk_Vocal_Exhale_11}}
 {id=WavePlayer_227, properties {soundWaveAssetPtr=&Robert_Walk_Vocal_Exhale_12}}
 {id=WavePlayer_228, properties {soundWaveAssetPtr=&Robert_Walk_Vocal_Exhale_13}}
 {id=WavePlayer_229, properties {soundWaveAssetPtr=&Robert_Walk_Vocal_Exhale_14}}
 {id=WavePlayer_230, properties {soundWaveAssetPtr=&Robert_Walk_Vocal_Exhale_15}}
 {id=WavePlayer_231, properties {soundWaveAssetPtr=&Robert_Walk_Vocal_Exhale_01}}
 {id=WavePlayer_232, properties {soundWaveAssetPtr=&Robert_Walk_Vocal_Exhale_02}}
 {id=WavePlayer_233, properties {soundWaveAssetPtr=&Robert_Walk_Vocal_Exhale_03}}
 {id=WavePlayer_234, properties {soundWaveAssetPtr=&Robert_Walk_Vocal_Exhale_04}}
 {id=WavePlayer_235, properties {soundWaveAssetPtr=&Robert_Walk_Vocal_Exhale_05}}
 {id=WavePlayer_236, properties {soundWaveAssetPtr=&Robert_Walk_Vocal_Exhale_06}}
 {id=WavePlayer_237, properties {soundWaveAssetPtr=&Robert_Walk_Vocal_Exhale_07}}
 {id=WavePlayer_238, properties {soundWaveAssetPtr=&Robert_Walk_Vocal_Exhale_08}}
 {id=WavePlayer_239, properties {soundWaveAssetPtr=&Robert_Walk_Vocal_Exhale_09}}
]
properties {
 maxDistance=12000
 priority=0.95
 soundClassObject=&Fort_Creatures_Vocalization_Tactical
 volumeMultiplier=1.25
 concurrencySet [/Irwin/Sounds/Robert/Robert_Vocal_Breathing_Concurrency_Owner, /Irwin/Sounds/Robert/Robert_Vocal_Breathing_Concurrency_Global]
}
