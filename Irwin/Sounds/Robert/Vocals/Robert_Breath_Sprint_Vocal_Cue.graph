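# Sprint breathing vocal cue for the creature Robert, reusing walk exhale and idle chitter waves.
# IsLocal and IsRiddenByEnemy branches set loudness: reduced for local, full when enemy-ridden or unset, quietest otherwise.
# Each branch has identical weighted random pools favoring pitch- and volume-modulated exhales, with occasional chitters or silence, under shared predator vocalization attenuation.
# Owner and global breathing concurrency sets limit overlapping playback.
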SoundCue /Irwin/Sounds/Robert/Vocals/Robert_Breath_Sprint_Vocal_Cue
root=Branch_1
nodes [
 {id=Attenuation_10, children [Random_25], properties {attenuationSettings=/Irwin/Sounds/Settings/Predator_Vocalization_Medium_Att}}
 {id=Attenuation_8, children [Random_21], properties {attenuationSettings=&Predator_Vocalization_Medium_Att}}
 {id=Attenuation_9, children [Random_22], properties {attenuationSettings=&Predator_Vocalization_Medium_Att}}
 {id=Branch_1, children [Mixer_11,Branch_4,Branch_4], properties {boolParameterName=IsLocal}}
 {id=Branch_4, children [Mixer_10,Mixer_9,Mixer_10], properties {boolParameterName=IsRiddenByEnemy}}
 {id=Mixer_10, children [Attenuation_9], properties {inputVolume [1]}}
 {id=Mixer_11, children [Attenuation_8], properties {inputVolume [0.7]}}
 {id=Mixer_9, children [Attenuation_10], properties {inputVolume [0.3]}}
 {id=Modulator_2, children [Random_24], properties {pitchMax=1.2, pitchMin=1, volumeMin=0.1}}
 {id=Modulator_4, children [Random_11], properties {pitchMax=1.2, pitchMin=1, volumeMin=0.1}}
 {id=Modulator_5, children [Random_27], properties {pitchMax=1.2, pitchMin=1, volumeMin=0.1}}
 {
  id=Random_10
  children [
   WavePlayer_199
   WavePlayer_200
   WavePlayer_201
   WavePlayer_202
   WavePlayer_203
   WavePlayer_204
   WavePlayer_205
   WavePlayer_206
   WavePlayer_207
   WavePlayer_208
   WavePlayer_209
   WavePlayer_210
   WavePlayer_211
   WavePlayer_212
  ]
  properties {bShouldExcludeFromBranchCulling=true, weights [1,1,1,1,1,1,1,1,1,1,1,1,1,1]}
 }
 {
  id=Random_11
  children [
   WavePlayer_100
   WavePlayer_101
   WavePlayer_102
   WavePlayer_103
   WavePlayer_104
   WavePlayer_105
   WavePlayer_106
   WavePlayer_107
   WavePlayer_108
   WavePlayer_109
   WavePlayer_110
   WavePlayer_111
   WavePlayer_112
   WavePlayer_113
   WavePlayer_114
  ]
  properties {bShouldExcludeFromBranchCulling=true, weights [1,1,1,1,1,1,1,1,1,1,1,1,1,1,1]}
 }
 {
  id=Random_21
  children [Modulator_4,Modulator_4,Modulator_4,Modulator_4,Modulator_4,Modulator_4,Random_10,null,null,null]
  properties {bShouldExcludeFromBranchCulling=true, weights [1,1,1,3,1,1,1,1,1,1]}
 }
 {
  id=Random_22
  children [Modulator_2,Modulator_2,Modulator_2,Modulator_2,Modulator_2,Modulator_2,Random_23,null,null,null]
  properties {bShouldExcludeFromBranchCulling=true, weights [1,1,1,3,1,1,1,1,1,1]}
 }
 {
  id=Random_23
  children [
   WavePlayer_213
   WavePlayer_214
   WavePlayer_215
   WavePlayer_216
   WavePlayer_217
   WavePlayer_218
   WavePlayer_219
   WavePlayer_220
   WavePlayer_221
   WavePlayer_222
   WavePlayer_223
   WavePlayer_224
   WavePlayer_225
   WavePlayer_226
  ]
  properties {bShouldExcludeFromBranchCulling=true, weights [1,1,1,1,1,1,1,1,1,1,1,1,1,1]}
 }
 {
  id=Random_24
  children [
   WavePlayer_227
   WavePlayer_228
   WavePlayer_229
   WavePlayer_230
   WavePlayer_231
   WavePlayer_232
   WavePlayer_233
   WavePlayer_234
   WavePlayer_235
   WavePlayer_236
   WavePlayer_237
   WavePlayer_238
   WavePlayer_239
   WavePlayer_240
   WavePlayer_241
  ]
  properties {bShouldExcludeFromBranchCulling=true, weights [1,1,1,1,1,1,1,1,1,1,1,1,1,1,1]}
 }
 {
  id=Random_25
  children [Modulator_5,Modulator_5,Modulator_5,Modulator_5,Modulator_5,Modulator_5,Random_26,null,null,null]
  properties {bShouldExcludeFromBranchCulling=true, weights [1,1,1,3,1,1,1,1,1,1]}
 }
 {
  id=Random_26
  children [
   WavePlayer_242
   WavePlayer_243
   WavePlayer_244
   WavePlayer_245
   WavePlayer_246
   WavePlayer_247
   WavePlayer_248
   WavePlayer_249
   WavePlayer_250
   WavePlayer_251
   WavePlayer_252
   WavePlayer_253
   WavePlayer_254
   WavePlayer_255
  ]
  properties {bShouldExcludeFromBranchCulling=true, weights [1,1,1,1,1,1,1,1,1,1,1,1,1,1]}
 }
 {
  id=Random_27
  children [
   WavePlayer_256
   WavePlayer_257
   WavePlayer_258
   WavePlayer_259
   WavePlayer_260
   WavePlayer_261
   WavePlayer_262
   WavePlayer_263
   WavePlayer_264
   WavePlayer_265
   WavePlayer_266
   WavePlayer_267
   WavePlayer_268
   WavePlayer_269
   WavePlayer_270
  ]
  properties {bShouldExcludeFromBranchCulling=true, weights [1,1,1,1,1,1,1,1,1,1,1,1,1,1,1]}
 }
 {id=WavePlayer_100, properties {soundWaveAssetPtr=&/Robert_Walk_Vocal_Exhale_10}}
 {id=WavePlayer_101, properties {soundWaveAssetPtr=&/Robert_Walk_Vocal_Exhale_11}}
 {id=WavePlayer_102, properties {soundWaveAssetPtr=&/Robert_Walk_Vocal_Exhale_12}}
 {id=WavePlayer_103, properties {soundWaveAssetPtr=&/Robert_Walk_Vocal_Exhale_13}}
 {id=WavePlayer_104, properties {soundWaveAssetPtr=&/Robert_Walk_Vocal_Exhale_14}}
 {id=WavePlayer_105, properties {soundWaveAssetPtr=&/Robert_Walk_Vocal_Exhale_15}}
 {id=WavePlayer_106, properties {soundWaveAssetPtr=&/Robert_Walk_Vocal_Exhale_01}}
 {id=WavePlayer_107, properties {soundWaveAssetPtr=&/Robert_Walk_Vocal_Exhale_02}}
 {id=WavePlayer_108, properties {soundWaveAssetPtr=&/Robert_Walk_Vocal_Exhale_03}}
 {id=WavePlayer_109, properties {soundWaveAssetPtr=&/Robert_Walk_Vocal_Exhale_04}}
 {id=WavePlayer_110, properties {soundWaveAssetPtr=&/Robert_Walk_Vocal_Exhale_05}}
 {id=WavePlayer_111, properties {soundWaveAssetPtr=&/Robert_Walk_Vocal_Exhale_06}}
 {id=WavePlayer_112, properties {soundWaveAssetPtr=&/Robert_Walk_Vocal_Exhale_07}}
 {id=WavePlayer_113, properties {soundWaveAssetPtr=&/Robert_Walk_Vocal_Exhale_08}}
 {id=WavePlayer_114, properties {soundWaveAssetPtr=&/Robert_Walk_Vocal_Exhale_09}}
 {id=WavePlayer_199, properties {soundWaveAssetPtr=&/Robert_Idle_Vocal_Chitter_10}}
 {id=WavePlayer_200, properties {soundWaveAssetPtr=&/Robert_Idle_Vocal_Chitter_11}}
 {id=WavePlayer_201, properties {soundWaveAssetPtr=&/Robert_Idle_Vocal_Chitter_15}}
 {id=WavePlayer_202, properties {soundWaveAssetPtr=&/Robert_Idle_Vocal_Chitter_16}}
 {id=WavePlayer_203, properties {soundWaveAssetPtr=&/Robert_Idle_Vocal_Chitter_17}}
 {id=WavePlayer_204, properties {soundWaveAssetPtr=&/Robert_Idle_Vocal_Chitter_21}}
 {id=WavePlayer_205, properties {soundWaveAssetPtr=&/Robert_Idle_Vocal_Chitter_22}}
 {id=WavePlayer_206, properties {soundWaveAssetPtr=&/Robert_Idle_Vocal_Chitter_01}}
 {id=WavePlayer_207, properties {soundWaveAssetPtr=&/Robert_Idle_Vocal_Chitter_02}}
 {id=WavePlayer_208, properties {soundWaveAssetPtr=&/Robert_Idle_Vocal_Chitter_04}}
 {id=WavePlayer_209, properties {soundWaveAssetPtr=&/Robert_Idle_Vocal_Chitter_06}}
 {id=WavePlayer_210, properties {soundWaveAssetPtr=&/Robert_Idle_Vocal_Chitter_07}}
 {id=WavePlayer_211, properties {soundWaveAssetPtr=&/Robert_Idle_Vocal_Chitter_08}}
 {id=WavePlayer_212, properties {soundWaveAssetPtr=&/Robert_Idle_Vocal_Chitter_09}}
 {id=WavePlayer_213, properties {soundWaveAssetPtr=&Robert_Idle_Vocal_Chitter_10}}
 {id=WavePlayer_214, properties {soundWaveAssetPtr=&Robert_Idle_Vocal_Chitter_11}}
 {id=WavePlayer_215, properties {soundWaveAssetPtr=&Robert_Idle_Vocal_Chitter_15}}
 {id=WavePlayer_216, properties {soundWaveAssetPtr=&Robert_Idle_Vocal_Chitter_16}}
 {id=WavePlayer_217, properties {soundWaveAssetPtr=&Robert_Idle_Vocal_Chitter_17}}
 {id=WavePlayer_218, properties {soundWaveAssetPtr=&Robert_Idle_Vocal_Chitter_21}}
 {id=WavePlayer_219, properties {soundWaveAssetPtr=&Robert_Idle_Vocal_Chitter_22}}
 {id=WavePlayer_220, properties {soundWaveAssetPtr=&Robert_Idle_Vocal_Chitter_01}}
 {id=WavePlayer_221, properties {soundWaveAssetPtr=&Robert_Idle_Vocal_Chitter_02}}
 {id=WavePlayer_222, properties {soundWaveAssetPtr=&Robert_Idle_Vocal_Chitter_04}}
 {id=WavePlayer_223, properties {soundWaveAssetPtr=&Robert_Idle_Vocal_Chitter_06}}
 {id=WavePlayer_224, properties {soundWaveAssetPtr=&Robert_Idle_Vocal_Chitter_07}}
 {id=WavePlayer_225, properties {soundWaveAssetPtr=&Robert_Idle_Vocal_Chitter_08}}
 {id=WavePlayer_226, properties {soundWaveAssetPtr=&Robert_Idle_Vocal_Chitter_09}}
 {id=WavePlayer_227, properties {soundWaveAssetPtr=&Robert_Walk_Vocal_Exhale_10}}
 {id=WavePlayer_228, properties {soundWaveAssetPtr=&Robert_Walk_Vocal_Exhale_11}}
 {id=WavePlayer_229, properties {soundWaveAssetPtr=&Robert_Walk_Vocal_Exhale_12}}
 {id=WavePlayer_230, properties {soundWaveAssetPtr=&Robert_Walk_Vocal_Exhale_13}}
 {id=WavePlayer_231, properties {soundWaveAssetPtr=&Robert_Walk_Vocal_Exhale_14}}
 {id=WavePlayer_232, properties {soundWaveAssetPtr=&Robert_Walk_Vocal_Exhale_15}}
 {id=WavePlayer_233, properties {soundWaveAssetPtr=&Robert_Walk_Vocal_Exhale_01}}
 {id=WavePlayer_234, properties {soundWaveAssetPtr=&Robert_Walk_Vocal_Exhale_02}}
 {id=WavePlayer_235, properties {soundWaveAssetPtr=&Robert_Walk_Vocal_Exhale_03}}
 {id=WavePlayer_236, properties {soundWaveAssetPtr=&Robert_Walk_Vocal_Exhale_04}}
 {id=WavePlayer_237, properties {soundWaveAssetPtr=&Robert_Walk_Vocal_Exhale_05}}
 {id=WavePlayer_238, properties {soundWaveAssetPtr=&Robert_Walk_Vocal_Exhale_06}}
 {id=WavePlayer_239, properties {soundWaveAssetPtr=&Robert_Walk_Vocal_Exhale_07}}
 {id=WavePlayer_240, properties {soundWaveAssetPtr=&Robert_Walk_Vocal_Exhale_08}}
 {id=WavePlayer_241, properties {soundWaveAssetPtr=&Robert_Walk_Vocal_Exhale_09}}
 {id=WavePlayer_242, properties {soundWaveAssetPtr=&Robert_Idle_Vocal_Chitter_10}}
 {id=WavePlayer_243, properties {soundWaveAssetPtr=&Robert_Idle_Vocal_Chitter_11}}
 {id=WavePlayer_244, properties {soundWaveAssetPtr=&Robert_Idle_Vocal_Chitter_15}}
 {id=WavePlayer_245, properties {soundWaveAssetPtr=&Robert_Idle_Vocal_Chitter_16}}
 {id=WavePlayer_246, properties {soundWaveAssetPtr=&Robert_Idle_Vocal_Chitter_17}}
 {id=WavePlayer_247, properties {soundWaveAssetPtr=&Robert_Idle_Vocal_Chitter_21}}
 {id=WavePlayer_248, properties {soundWaveAssetPtr=&Robert_Idle_Vocal_Chitter_22}}
 {id=WavePlayer_249, properties {soundWaveAssetPtr=&Robert_Idle_Vocal_Chitter_01}}
 {id=WavePlayer_250, properties {soundWaveAssetPtr=&Robert_Idle_Vocal_Chitter_02}}
 {id=WavePlayer_251, properties {soundWaveAssetPtr=&Robert_Idle_Vocal_Chitter_04}}
 {id=WavePlayer_252, properties {soundWaveAssetPtr=&Robert_Idle_Vocal_Chitter_06}}
 {id=WavePlayer_253, properties {soundWaveAssetPtr=&Robert_Idle_Vocal_Chitter_07}}
 {id=WavePlayer_254, properties {soundWaveAssetPtr=&Robert_Idle_Vocal_Chitter_08}}
 {id=WavePlayer_255, properties {soundWaveAssetPtr=&Robert_Idle_Vocal_Chitter_09}}
 {id=WavePlayer_256, properties {soundWaveAssetPtr=&Robert_Walk_Vocal_Exhale_10}}
 {id=WavePlayer_257, properties {soundWaveAssetPtr=&Robert_Walk_Vocal_Exhale_11}}
 {id=WavePlayer_258, properties {soundWaveAssetPtr=&Robert_Walk_Vocal_Exhale_12}}
 {id=WavePlayer_259, properties {soundWaveAssetPtr=&Robert_Walk_Vocal_Exhale_13}}
 {id=WavePlayer_260, properties {soundWaveAssetPtr=&Robert_Walk_Vocal_Exhale_14}}
 {id=WavePlayer_261, properties {soundWaveAssetPtr=&Robert_Walk_Vocal_Exhale_15}}
 {id=WavePlayer_262, properties {soundWaveAssetPtr=&Robert_Walk_Vocal_Exhale_01}}
 {id=WavePlayer_263, properties {soundWaveAssetPtr=&Robert_Walk_Vocal_Exhale_02}}
 {id=WavePlayer_264, properties {soundWaveAssetPtr=&Robert_Walk_Vocal_Exhale_03}}
 {id=WavePlayer_265, properties {soundWaveAssetPtr=&Robert_Walk_Vocal_Exhale_04}}
 {id=WavePlayer_266, properties {soundWaveAssetPtr=&Robert_Walk_Vocal_Exhale_05}}
 {id=WavePlayer_267, properties {soundWaveAssetPtr=&Robert_Walk_Vocal_Exhale_06}}
 {id=WavePlayer_268, properties {soundWaveAssetPtr=&Robert_Walk_Vocal_Exhale_07}}
 {id=WavePlayer_269, properties {soundWaveAssetPtr=&Robert_Walk_Vocal_Exhale_08}}
 {id=WavePlayer_270, properties {soundWaveAssetPtr=&Robert_Walk_Vocal_Exhale_09}}
]
properties {
 maxDistance=12000
 priority=0.95
 soundClassObject=&Fort_Creatures_Vocalization_Tactical
 volumeMultiplier=1.25
 concurrencyOverrides {bLimitToOwner=true, maxCount=1, resolutionRule=StopLowestPriorityThenPreventNew}
 concurrencySet [/Irwin/Sounds/Robert/Robert_Vocal_Breathing_Concurrency_Owner, /Irwin/Sounds/Robert/Robert_Vocal_Breathing_Concurrency_Global]
}
