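# Sound cue for Robert's walking breath vocals, playing random exhales, occasional idle chitters or silence.
# IsLocal and IsRiddenByEnemy branches select one of three identical attenuated chains that differ in mixer volume: medium for local, full for enemy-ridden or unset, quietest otherwise.
# Each chain uses a weighted Random with null slots for silence, modulating pitch and volume only on exhales.
# Owner and global concurrency sets limit overlapping breaths.
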SoundCue /Irwin/Sounds/Robert/Vocals/Robert_Breath_Walk_Vocal_Cue
root=Branch_3
nodes [
 {id=Attenuation_2, children [Random_3], properties {attenuationSettings=/Irwin/Sounds/Settings/Predator_Vocalization_Medium_Att}}
 {id=Attenuation_3, children [Random_7], properties {attenuationSettings=&Predator_Vocalization_Medium_Att}}
 {id=Attenuation_8, children [Random_21], properties {attenuationSettings=&Predator_Vocalization_Medium_Att}}
 {id=Branch_1, children [Mixer_7,Mixer_4,Mixer_7], properties {boolParameterName=IsRiddenByEnemy}}
 {id=Branch_3, children [Mixer_8,Branch_1,Branch_1], properties {boolParameterName=IsLocal}}
 {id=Mixer_4, children [Attenuation_8], properties {inputVolume [0.3]}}
 {id=Mixer_7, children [Attenuation_3], properties {inputVolume [1]}}
 {id=Mixer_8, children [Attenuation_2], properties {inputVolume [0.7]}}
 {id=Modulator_1, children [Random_20], properties {pitchMax=1.2, pitchMin=1, volumeMin=0.1}}
 {id=Modulator_4, children [Random_6], properties {pitchMax=1.2, pitchMin=1, volumeMin=0.1}}
 {id=Modulator_5, children [Random_23], properties {pitchMax=1.2, pitchMin=1, volumeMin=0.1}}
 {
  id=Random_20
  children [
   WavePlayer_92
   WavePlayer_93
   WavePlayer_94
   WavePlayer_95
   WavePlayer_96
   WavePlayer_97
   WavePlayer_98
   WavePlayer_99
   WavePlayer_204
   WavePlayer_205
   WavePlayer_206
   WavePlayer_207
   WavePlayer_208
   WavePlayer_209
   WavePlayer_210
  ]
  properties {bShouldExcludeFromBranchCulling=true, weights [1,1,1,1,1,1,1,1,1,1,1,1,1,1,1]}
 }
 {
  id=Random_21
  children [Modulator_5,Modulator_5,Modulator_5,Modulator_5,Modulator_5,Modulator_5,Random_22,null,null,null]
  properties {bShouldExcludeFromBranchCulling=true, weights [1,1,1,3,1,1,1,1,1,1]}
 }
 {
  id=Random_22
  children [
   WavePlayer_211
   WavePlayer_212
   WavePlayer_213
   WavePlayer_214
   WavePlayer_215
   WavePlayer_216
   WavePlayer_217
   WavePlayer_218
   WavePlayer_219
   WavePlayer_220
   WavePlayer_221
   WavePlayer_222
   WavePlayer_223
   WavePlayer_224
  ]
  properties {bShouldExcludeFromBranchCulling=true, weights [1,1,1,1,1,1,1,1,1,1,1,1,1,1]}
 }
 {
  id=Random_23
  children [
   WavePlayer_225
   WavePlayer_226
   WavePlayer_227
   WavePlayer_228
   WavePlayer_229
   WavePlayer_230
   WavePlayer_231
   WavePlayer_232
   WavePlayer_233
   WavePlayer_234
   WavePlayer_235
   WavePlayer_236
   WavePlayer_237
   WavePlayer_238
   WavePlayer_239
  ]
  properties {bShouldExcludeFromBranchCulling=true, weights [1,1,1,1,1,1,1,1,1,1,1,1,1,1,1]}
 }
 {
  id=Random_3
  children [Modulator_4,Modulator_4,Modulator_4,Modulator_4,Modulator_4,Modulator_4,Random_5,null,null,null]
  properties {bShouldExcludeFromBranchCulling=true, weights [1,1,1,3,1,1,1,1,1,1]}
 }
 {
  id=Random_5
  children [
   WavePlayer_66
   WavePlayer_67
   WavePlayer_68
   WavePlayer_69
   WavePlayer_70
   WavePlayer_71
   WavePlayer_72
   WavePlayer_73
   WavePlayer_74
   WavePlayer_75
   WavePlayer_76
   WavePlayer_77
   WavePlayer_78
   WavePlayer_79
  ]
  properties {bShouldExcludeFromBranchCulling=true, weights [1,1,1,1,1,1,1,1,1,1,1,1,1,1]}
 }
 {
  id=Random_6
  children [
   WavePlayer_17
   WavePlayer_18
   WavePlayer_21
   WavePlayer_22
   WavePlayer_23
   WavePlayer_24
   WavePlayer_25
   WavePlayer_26
   WavePlayer_29
   WavePlayer_30
   WavePlayer_32
   WavePlayer_52
   WavePlayer_53
   WavePlayer_54
   WavePlayer_57
  ]
  properties {bShouldExcludeFromBranchCulling=true, weights [1,1,1,1,1,1,1,1,1,1,1,1,1,1,1]}
 }
 {
  id=Random_7
  children [Modulator_1,Modulator_1,Modulator_1,Modulator_1,Modulator_1,Modulator_1,Random_8,null,null,null]
  properties {bShouldExcludeFromBranchCulling=true, weights [1,1,1,3,1,1,1,1,1,1]}
 }
 {
  id=Random_8
  children [
   WavePlayer_59
   WavePlayer_61
   WavePlayer_80
   WavePlayer_81
   WavePlayer_82
   WavePlayer_83
   WavePlayer_84
   WavePlayer_85
   WavePlayer_86
   WavePlayer_87
   WavePlayer_88
   WavePlayer_89
   WavePlayer_90
   WavePlayer_91
  ]
  properties {bShouldExcludeFromBranchCulling=true, weights [1,1,1,1,1,1,1,1,1,1,1,1,1,1]}
 }
 {id=WavePlayer_17, properties {soundWaveAssetPtr=&/Robert_Walk_Vocal_Exhale_10}}
 {id=WavePlayer_18, properties {soundWaveAssetPtr=&/Robert_Walk_Vocal_Exhale_11}}
 {id=WavePlayer_204, properties {soundWaveAssetPtr=&/Robert_Walk_Vocal_Exhale_03}}
 {id=WavePlayer_205, properties {soundWaveAssetPtr=&/Robert_Walk_Vocal_Exhale_04}}
 {id=WavePlayer_206, properties {soundWaveAssetPtr=&/Robert_Walk_Vocal_Exhale_05}}
 {id=WavePlayer_207, properties {soundWaveAssetPtr=&/Robert_Walk_Vocal_Exhale_06}}
 {id=WavePlayer_208, properties {soundWaveAssetPtr=&/Robert_Walk_Vocal_Exhale_07}}
 {id=WavePlayer_209, properties {soundWaveAssetPtr=&/Robert_Walk_Vocal_Exhale_08}}
 {id=WavePlayer_21, properties {soundWaveAssetPtr=&/Robert_Walk_Vocal_Exhale_12}}
 {id=WavePlayer_210, properties {soundWaveAssetPtr=&/Robert_Walk_Vocal_Exhale_09}}
 {id=WavePlayer_211, properties {soundWaveAssetPtr=&/Robert_Idle_Vocal_Chitter_10}}
 {id=WavePlayer_212, properties {soundWaveAssetPtr=&/Robert_Idle_Vocal_Chitter_11}}
 {id=WavePlayer_213, properties {soundWaveAssetPtr=&/Robert_Idle_Vocal_Chitter_15}}
 {id=WavePlayer_214, properties {soundWaveAssetPtr=&/Robert_Idle_Vocal_Chitter_16}}
 {id=WavePlayer_215, properties {soundWaveAssetPtr=&/Robert_Idle_Vocal_Chitter_17}}
 {id=WavePlayer_216, properties {soundWaveAssetPtr=&/Robert_Idle_Vocal_Chitter_21}}
 {id=WavePlayer_217, properties {soundWaveAssetPtr=&/Robert_Idle_Vocal_Chitter_22}}
 {id=WavePlayer_218, properties {soundWaveAssetPtr=&/Robert_Idle_Vocal_Chitter_01}}
 {id=WavePlayer_219, properties {soundWaveAssetPtr=&/Robert_Idle_Vocal_Chitter_02}}
 {id=WavePlayer_22, properties {soundWaveAssetPtr=&/Robert_Walk_Vocal_Exhale_13}}
 {id=WavePlayer_220, properties {soundWaveAssetPtr=&/Robert_Idle_Vocal_Chitter_04}}
 {id=WavePlayer_221, properties {soundWaveAssetPtr=&/Robert_Idle_Vocal_Chitter_06}}
 {id=WavePlayer_222, properties {soundWaveAssetPtr=&/Robert_Idle_Vocal_Chitter_07}}
 {id=WavePlayer_223, properties {soundWaveAssetPtr=&/Robert_Idle_Vocal_Chitter_08}}
 {id=WavePlayer_224, properties {soundWaveAssetPtr=&/Robert_Idle_Vocal_Chitter_09}}
 {id=WavePlayer_225, properties {soundWaveAssetPtr=&Robert_Walk_Vocal_Exhale_10}}
 {id=WavePlayer_226, properties {soundWaveAssetPtr=&Robert_Walk_Vocal_Exhale_11}}
 {id=WavePlayer_227, properties {soundWaveAssetPtr=&Robert_Walk_Vocal_Exhale_12}}
 {id=WavePlayer_228, properties {soundWaveAssetPtr=&Robert_Walk_Vocal_Exhale_13}}
 {id=WavePlayer_229, properties {soundWaveAssetPtr=&/Robert_Walk_Vocal_Exhale_14}}
 {id=WavePlayer_23, properties {soundWaveAssetPtr=&Robert_Walk_Vocal_Exhale_14}}
 {id=WavePlayer_230, properties {soundWaveAssetPtr=&/Robert_Walk_Vocal_Exhale_15}}
 {id=WavePlayer_231, properties {soundWaveAssetPtr=&/Robert_Walk_Vocal_Exhale_01}}
 {id=WavePlayer_232, properties {soundWaveAssetPtr=&/Robert_Walk_Vocal_Exhale_02}}
 {id=WavePlayer_233, properties {soundWaveAssetPtr=&Robert_Walk_Vocal_Exhale_03}}
 {id=WavePlayer_234, properties {soundWaveAssetPtr=&Robert_Walk_Vocal_Exhale_04}}
 {id=WavePlayer_235, properties {soundWaveAssetPtr=&Robert_Walk_Vocal_Exhale_05}}
 {id=WavePlayer_236, properties {soundWaveAssetPtr=&Robert_Walk_Vocal_Exhale_06}}
 {id=WavePlayer_237, properties {soundWaveAssetPtr=&Robert_Walk_Vocal_Exhale_07}}
 {id=WavePlayer_238, properties {soundWaveAssetPtr=&Robert_Walk_Vocal_Exhale_08}}
 {id=WavePlayer_239, properties {soundWaveAssetPtr=&Robert_Walk_Vocal_Exhale_09}}
 {id=WavePlayer_24, properties {soundWaveAssetPtr=&Robert_Walk_Vocal_Exhale_15}}
 {id=WavePlayer_25, properties {soundWaveAssetPtr=&Robert_Walk_Vocal_Exhale_01}}
 {id=WavePlayer_26, properties {soundWaveAssetPtr=&Robert_Walk_Vocal_Exhale_02}}
 {id=WavePlayer_29, properties {soundWaveAssetPtr=&Robert_Walk_Vocal_Exhale_03}}
 {id=WavePlayer_30, properties {soundWaveAssetPtr=&Robert_Walk_Vocal_Exhale_04}}
 {id=WavePlayer_32, properties {soundWaveAssetPtr=&Robert_Walk_Vocal_Exhale_05}}
 {id=WavePlayer_52, properties {soundWaveAssetPtr=&Robert_Walk_Vocal_Exhale_06}}
 {id=WavePlayer_53, properties {soundWaveAssetPtr=&Robert_Walk_Vocal_Exhale_07}}
 {id=WavePlayer_54, properties {soundWaveAssetPtr=&Robert_Walk_Vocal_Exhale_08}}
 {id=WavePlayer_57, properties {soundWaveAssetPtr=&Robert_Walk_Vocal_Exhale_09}}
 {id=WavePlayer_59, properties {soundWaveAssetPtr=&Robert_Idle_Vocal_Chitter_10}}
 {id=WavePlayer_61, properties {soundWaveAssetPtr=&Robert_Idle_Vocal_Chitter_11}}
 {id=WavePlayer_66, properties {soundWaveAssetPtr=&Robert_Idle_Vocal_Chitter_10}}
 {id=WavePlayer_67, properties {soundWaveAssetPtr=&Robert_Idle_Vocal_Chitter_11}}
 {id=WavePlayer_68, properties {soundWaveAssetPtr=&Robert_Idle_Vocal_Chitter_15}}
 {id=WavePlayer_69, properties {soundWaveAssetPtr=&Robert_Idle_Vocal_Chitter_16}}
 {id=WavePlayer_70, properties {soundWaveAssetPtr=&Robert_Idle_Vocal_Chitter_17}}
 {id=WavePlayer_71, properties {soundWaveAssetPtr=&Robert_Idle_Vocal_Chitter_21}}
 {id=WavePlayer_72, properties {soundWaveAssetPtr=&Robert_Idle_Vocal_Chitter_22}}
 {id=WavePlayer_73, properties {soundWaveAssetPtr=&Robert_Idle_Vocal_Chitter_01}}
 {id=WavePlayer_74, properties {soundWaveAssetPtr=&Robert_Idle_Vocal_Chitter_02}}
 {id=WavePlayer_75, properties {soundWaveAssetPtr=&Robert_Idle_Vocal_Chitter_04}}
 {id=WavePlayer_76, properties {soundWaveAssetPtr=&Robert_Idle_Vocal_Chitter_06}}
 {id=WavePlayer_77, properties {soundWaveAssetPtr=&Robert_Idle_Vocal_Chitter_07}}
 {id=WavePlayer_78, properties {soundWaveAssetPtr=&Robert_Idle_Vocal_Chitter_08}}
 {id=WavePlayer_79, properties {soundWaveAssetPtr=&Robert_Idle_Vocal_Chitter_09}}
 {id=WavePlayer_80, properties {soundWaveAssetPtr=&Robert_Idle_Vocal_Chitter_15}}
 {id=WavePlayer_81, properties {soundWaveAssetPtr=&Robert_Idle_Vocal_Chitter_16}}
 {id=WavePlayer_82, properties {soundWaveAssetPtr=&Robert_Idle_Vocal_Chitter_17}}
 {id=WavePlayer_83, properties {soundWaveAssetPtr=&Robert_Idle_Vocal_Chitter_21}}
 {id=WavePlayer_84, properties {soundWaveAssetPtr=&Robert_Idle_Vocal_Chitter_22}}
 {id=WavePlayer_85, properties {soundWaveAssetPtr=&Robert_Idle_Vocal_Chitter_01}}
 {id=WavePlayer_86, properties {soundWaveAssetPtr=&Robert_Idle_Vocal_Chitter_02}}
 {id=WavePlayer_87, properties {soundWaveAssetPtr=&Robert_Idle_Vocal_Chitter_04}}
 {id=WavePlayer_88, properties {soundWaveAssetPtr=&Robert_Idle_Vocal_Chitter_06}}
 {id=WavePlayer_89, properties {soundWaveAssetPtr=&Robert_Idle_Vocal_Chitter_07}}
 {id=WavePlayer_90, properties {soundWaveAssetPtr=&Robert_Idle_Vocal_Chitter_08}}
 {id=WavePlayer_91, properties {soundWaveAssetPtr=&Robert_Idle_Vocal_Chitter_09}}
 {id=WavePlayer_92, properties {soundWaveAssetPtr=&Robert_Walk_Vocal_Exhale_10}}
 {id=WavePlayer_93, properties {soundWaveAssetPtr=&Robert_Walk_Vocal_Exhale_11}}
 {id=WavePlayer_94, properties {soundWaveAssetPtr=&Robert_Walk_Vocal_Exhale_12}}
 {id=WavePlayer_95, properties {soundWaveAssetPtr=&Robert_Walk_Vocal_Exhale_13}}
 {id=WavePlayer_96, properties {soundWaveAssetPtr=&Robert_Walk_Vocal_Exhale_14}}
 {id=WavePlayer_97, properties {soundWaveAssetPtr=&Robert_Walk_Vocal_Exhale_15}}
 {id=WavePlayer_98, properties {soundWaveAssetPtr=&Robert_Walk_Vocal_Exhale_01}}
 {id=WavePlayer_99, properties {soundWaveAssetPtr=&Robert_Walk_Vocal_Exhale_02}}
]
properties {
 maxDistance=12000
 priority=0.95
 soundClassObject=&Fort_Creatures_Vocalization_Tactical
 volumeMultiplier=1.5
 concurrencyOverrides {bLimitToOwner=true, maxCount=1, resolutionRule=StopLowestPriorityThenPreventNew}
 concurrencySet [/Irwin/Sounds/Robert/Robert_Vocal_Breathing_Concurrency_Owner, /Irwin/Sounds/Robert/Robert_Vocal_Breathing_Concurrency_Global]
}
