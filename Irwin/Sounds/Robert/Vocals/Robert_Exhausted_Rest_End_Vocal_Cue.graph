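# Sound cue for the creature Robert's vocal at the end of an exhausted rest, using predator vocalization attenuation and the creature vocalization sound class.
# Branches on IsLocal, then IsRiddenByEnemy, route through separate mixers that all share one random node over three equally weighted waves.
# Every mixer has the same volume, so all branches currently sound alike.
# A concurrency override caps simultaneous instances, and the random node is excluded from branch culling.
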SoundCue /Irwin/Sounds/Robert/Vocals/Robert_Exhausted_Rest_End_Vocal_Cue
root=Branch_2
nodes [
 {id=Branch_0, children [Mixer_2,Mixer_1,Mixer_2], properties {boolParameterName=IsRiddenByEnemy}}
 {id=Branch_2, children [Mixer_3,Branch_0,Branch_0], properties {boolParameterName=IsLocal}}
 {id=Mixer_1, children [Random_1], properties {inputVolume [1]}}
 {id=Mixer_2, children [Random_1], properties {inputVolume [1]}}
 {id=Mixer_3, children [Random_1], properties {inputVolume [1]}}
 {id=Random_1, children [WavePlayer_0,WavePlayer_1,WavePlayer_2], properties {bShouldExcludeFromBranchCulling=true, weights [1,1,1]}}
 {id=WavePlayer_0, properties {soundWaveAssetPtr=&/Robert_Exhausted_Vocal_Rest_End_01}}
 {id=WavePlayer_1, properties {soundWaveAssetPtr=&/Robert_Exhausted_Vocal_Rest_End_02}}
 {id=WavePlayer_2, properties {soundWaveAssetPtr=&/Robert_Exhausted_Vocal_Rest_End_03}}
]
properties {
 attenuationSettings=/Irwin/Sounds/Settings/Predator_Vocalization_Medium_Att
 bOverrideConcurrency=true
 maxDistance=12000
 soundClassObject=/Game/Sounds/FortSoundClassesAndModes/Fort_Creatures_Vocalization_Normal
 volumeMultiplier=1.5
 concurrencyOverrides {maxCount=3}
}
