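# Sound cue for Robert's looping exhausted-rest vocal, a creature vocalization with medium predator attenuation.
# An IsLocal branch picks the local path or passes to an IsRiddenByEnemy branch, which separates enemy-ridden from other playback.
# All three paths share one structure: a looping wave of the same name, an enveloper that ramps volume up from silence, and a single-input mixer.
# A concurrency override caps simultaneous instances.
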SoundCue /Irwin/Sounds/Robert/Vocals/Robert_Exhausted_Rest_Vocal_Loop_Cue
root=Branch_1
nodes [
 {id=Branch_0, children [Mixer_2,Mixer_1,Mixer_2], properties {boolParameterName=IsRiddenByEnemy}}
 {id=Branch_1, children [Mixer_3,Branch_0,Branch_0], properties {boolParameterName=IsLocal}}
 {
  id=Enveloper_0
  children [WavePlayer_3]
  properties {pitchCurve {editorCurveData {keys [{time=0}]}}, volumeCurve {editorCurveData {keys [<at 0 0>,<at 1 1>]}}}
 }
 {
  id=Enveloper_1
  children [WavePlayer_0]
  properties {pitchCurve {editorCurveData {keys [{time=0}]}}, volumeCurve {editorCurveData {keys [<at 0 0>,<at 1 1>]}}}
 }
 {
  id=Enveloper_2
  children [WavePlayer_1]
  properties {pitchCurve {editorCurveData {keys [{time=0}]}}, volumeCurve {editorCurveData {keys [<at 0 0>,<at 1 1>]}}}
 }
 {id=Mixer_1, children [Enveloper_2], properties {inputVolume [1]}}
 {id=Mixer_2, children [Enveloper_1], properties {inputVolume [1]}}
 {id=Mixer_3, children [Enveloper_0], properties {inputVolume [1]}}
 {id=WavePlayer_0, properties {bLooping=true, soundWaveAssetPtr=&/Robert_Exhausted_Vocal_Rest_Loop}}
 {id=WavePlayer_1, properties {bLooping=true, soundWaveAssetPtr=&Robert_Exhausted_Vocal_Rest_Loop}}
 {id=WavePlayer_3, properties {bLooping=true, soundWaveAssetPtr=&Robert_Exhausted_Vocal_Rest_Loop}}
]
properties {
 attenuationSettings=/Irwin/Sounds/Settings/Predator_Vocalization_Medium_Att
 bOverrideConcurrency=true
 maxDistance=12000
 soundClassObject=/Game/Sounds/FortSoundClassesAndModes/Fort_Creatures_Vocalization_Normal
 volumeMultiplier=1
 concurrencyOverrides {maxCount=3}
}
