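# Sound cue for Robert's exhausted vocal when stamina is lost, using predator vocalization attenuation and the creature vocalization sound class.
# An IsLocal branch, then an IsRiddenByEnemy branch for the non-local case, selects a mixer feeding an equal-weight random pick of three takes.
# All paths use the same three take names, so the branches add no audible variation as authored.
# Owner-limited concurrency allows one instance, stopping the oldest.
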SoundCue /Irwin/Sounds/Robert/Vocals/Robert_Exhausted_Stamina_Lost_Vocal_Cue
root=Branch_2
nodes [
 {id=Branch_0, children [Mixer_2,Mixer_1,Mixer_2], properties {boolParameterName=IsRiddenByEnemy}}
 {id=Branch_2, children [Mixer_3,Branch_0,Branch_0], properties {boolParameterName=IsLocal}}
 {id=Mixer_1, children [Random_2], properties {inputVolume [1]}}
 {id=Mixer_2, children [Random_0], properties {inputVolume [1]}}
 {id=Mixer_3, children [Random_1], properties {inputVolume [1]}}
 {id=Random_0, children [WavePlayer_0,WavePlayer_1,WavePlayer_2], properties {bShouldExcludeFromBranchCulling=true, weights [1,1,1]}}
 {id=Random_1, children [WavePlayer_9, WavePlayer_11, WavePlayer_12], properties {bShouldExcludeFromBranchCulling=true, weights [1,1,1]}}
 {id=Random_2, children [WavePlayer_3,WavePlayer_4,WavePlayer_5], properties {bShouldExcludeFromBranchCulling=true, weights [1,1,1]}}
 {id=WavePlayer_0, properties {soundWaveAssetPtr=&/Robert_Exhausted_Vocal_Stamina_Lost_01}}
 {id=WavePlayer_1, properties {soundWaveAssetPtr=&/Robert_Exhausted_Vocal_Stamina_Lost_02}}
 {id=WavePlayer_11, properties {soundWaveAssetPtr=&Robert_Exhausted_Vocal_Stamina_Lost_02}}
 {id=WavePlayer_12, properties {soundWaveAssetPtr=&/Robert_Exhausted_Vocal_Stamina_Lost_03}}
 {id=WavePlayer_2, properties {soundWaveAssetPtr=&Robert_Exhausted_Vocal_Stamina_Lost_03}}
 {id=WavePlayer_3, properties {soundWaveAssetPtr=&Robert_Exhausted_Vocal_Stamina_Lost_01}}
 {id=WavePlayer_4, properties {soundWaveAssetPtr=&Robert_Exhausted_Vocal_Stamina_Lost_02}}
 {id=WavePlayer_5, properties {soundWaveAssetPtr=&Robert_Exhausted_Vocal_Stamina_Lost_03}}
 {id=WavePlayer_9, properties {soundWaveAssetPtr=&Robert_Exhausted_Vocal_Stamina_Lost_01}}
]
properties {
 attenuationSettings=/Irwin/Sounds/Settings/Predator_Vocalization_Medium_Att
 bOverrideConcurrency=true
 maxDistance=12000
 soundClassObject=/Game/Sounds/FortSoundClassesAndModes/Fort_Creatures_Vocalization_Normal
 volumeMultiplier=1.5
 concurrencyOverrides {bLimitToOwner=true, maxCount=1, resolutionRule=StopOldest}
}
