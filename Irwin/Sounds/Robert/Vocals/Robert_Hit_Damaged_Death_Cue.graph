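# Sound cue for Robert's death vocal, playing a random derez vocal through predator vocalization attenuation and a creature vocalization sound class.
# A quality-level switch feeds the first two levels from a six-variant random pool and the last two from a three-variant subset, trimming variety by quality.
# Both random pools are excluded from branch culling, and a concurrency override caps simultaneous plays.
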SoundCue /Irwin/Sounds/Robert/Vocals/Robert_Hit_Damaged_Death_Cue
root=QualityLevel_0
nodes [
 {id=QualityLevel_0, children [Random_2,Random_2,Random_3,Random_3]}
 {
  id=Random_2
  children [WavePlayer_6, WavePlayer_7, WavePlayer_8, WavePlayer_9, WavePlayer_10, WavePlayer_11]
  properties {bShouldExcludeFromBranchCulling=true, weights [1,1,1,1,1,1]}
 }
 {id=Random_3, children [WavePlayer_12, WavePlayer_13, WavePlayer_14], properties {bShouldExcludeFromBranchCulling=true, weights [1,1,1]}}
 {id=WavePlayer_10, properties {soundWaveAssetPtr=&/Robert_Death_Derez_Vocal_05}}
 {id=WavePlayer_11, properties {soundWaveAssetPtr=&/Robert_Death_Derez_Vocal_06}}
 {id=WavePlayer_12, properties {soundWaveAssetPtr=&/Robert_Death_Derez_Vocal_01}}
 {id=WavePlayer_13, properties {soundWaveAssetPtr=&/Robert_Death_Derez_Vocal_02}}
 {id=WavePlayer_14, properties {soundWaveAssetPtr=&/Robert_Death_Derez_Vocal_03}}
 {id=WavePlayer_6, properties {soundWaveAssetPtr=&Robert_Death_Derez_Vocal_01}}
 {id=WavePlayer_7, properties {soundWaveAssetPtr=&Robert_Death_Derez_Vocal_02}}
 {id=WavePlayer_8, properties {soundWaveAssetPtr=&Robert_Death_Derez_Vocal_03}}
 {id=WavePlayer_9, properties {soundWaveAssetPtr=&/Robert_Death_Derez_Vocal_04}}
]
properties {
 attenuationSettings=/Irwin/Sounds/Settings/Predator_Vocalization_Medium_Att
 bOverrideConcurrency=true
 maxDistance=12000
 soundClassObject=&Fort_Creatures_Vocalization_Tactical
 volumeMultiplier=1.5
 concurrencyOverrides {maxCount=3}
}
