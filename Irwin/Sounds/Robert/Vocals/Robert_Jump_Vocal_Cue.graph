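# Sound cue for jump vocals of the Robert creature, using predator vocalization attenuation and allowing one instance per owner.
# An IsLocal branch plays random walk exhales when true, otherwise deferring to an IsRiddenByEnemy branch.
# That branch plays random light-damage grunts at full volume when true or unset, and half volume when false.
# Empty slots in the exhale and half-volume pools make those vocals sometimes silent.
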SoundCue /Irwin/Sounds/Robert/Vocals/Robert_Jump_Vocal_Cue
root=Branch_2
nodes [
 {id=Branch_0, children [Mixer_2,Mixer_1,Mixer_2], properties {boolParameterName=IsRiddenByEnemy}}
 {id=Branch_2, children [Mixer_3,Branch_0,Branch_0], properties {boolParameterName=IsLocal}}
 {id=Mixer_1, children [Random_3], properties {inputVolume [0.5]}}
 {id=Mixer_2, children [Random_1], properties {inputVolume [1]}}
 {id=Mixer_3, children [Random_2], properties {inputVolume [1]}}
 {
  id=Random_1
  children [WavePlayer_8, WavePlayer_9, WavePlayer_10, WavePlayer_19, WavePlayer_20, WavePlayer_21, WavePlayer_22, WavePlayer_23]
  properties {bShouldExcludeFromBranchCulling=true, weights [1,1,1,1,1,1,1,1]}
 }
 {
  id=Random_2
  children [
   WavePlayer_102
   WavePlayer_104
   WavePlayer_106
   WavePlayer_107
   WavePlayer_108
   WavePlayer_109
   WavePlayer_111
   WavePlayer_112
   WavePlayer_113
   WavePlayer_114
   null
   null
   null
   null
   null
   null
   null
  ]
  properties {bShouldExcludeFromBranchCulling=true, weights [1,1,1,1,1,1,1,1,1,1,1,1,1,1,1,1,1]}
 }
 {
  id=Random_3
  children [
   WavePlayer_24
   WavePlayer_25
   WavePlayer_26
   WavePlayer_27
   WavePlayer_28
   WavePlayer_29
   WavePlayer_30
   WavePlayer_31
   null
   null
   null
   null
   null
   null
   null
  ]
  properties {bShouldExcludeFromBranchCulling=true, weights [1,1,1,1,1,1,1,1,1,1,1,1,1,1,1]}
 }
 {id=WavePlayer_10, properties {soundWaveAssetPtr=&/Robert_Hit_Damaged_Light_Vocal_03}}
 {id=WavePlayer_102, properties {soundWaveAssetPtr=&/Robert_Walk_Vocal_Exhale_12}}
 {id=WavePlayer_104, properties {soundWaveAssetPtr=&/Robert_Walk_Vocal_Exhale_14}}
 {id=WavePlayer_106, properties {soundWaveAssetPtr=&/Robert_Walk_Vocal_Exhale_01}}
 {id=WavePlayer_107, properties {soundWaveAssetPtr=&/Robert_Walk_Vocal_Exhale_02}}
 {id=WavePlayer_108, properties {soundWaveAssetPtr=&/Robert_Walk_Vocal_Exhale_03}}
 {id=WavePlayer_109, properties {soundWaveAssetPtr=&/Robert_Walk_Vocal_Exhale_04}}
 {id=WavePlayer_111, properties {soundWaveAssetPtr=&/Robert_Walk_Vocal_Exhale_06}}
 {id=WavePlayer_112, properties {soundWaveAssetPtr=&/Robert_Walk_Vocal_Exhale_07}}
 {id=WavePlayer_113, properties {soundWaveAssetPtr=&/Robert_Walk_Vocal_Exhale_08}}
 {id=WavePlayer_114, properties {soundWaveAssetPtr=&/Robert_Walk_Vocal_Exhale_09}}
 {id=WavePlayer_19, properties {soundWaveAssetPtr=&/Robert_Hit_Damaged_Light_Vocal_04}}
 {id=WavePlayer_20, properties {soundWaveAssetPtr=&/Robert_Hit_Damaged_Light_Vocal_05}}
 {id=WavePlayer_21, properties {soundWaveAssetPtr=&/Robert_Hit_Damaged_Light_Vocal_06}}
 {id=WavePlayer_22, properties {soundWaveAssetPtr=&/Robert_Hit_Damaged_Light_Vocal_07}}
 {id=WavePlayer_23, properties {soundWaveAssetPtr=&/Robert_Hit_Damaged_Light_Vocal_08}}
 {id=WavePlayer_24, properties {soundWaveAssetPtr=&/Robert_Hit_Damaged_Light_Vocal_01}}
 {id=WavePlayer_25, properties {soundWaveAssetPtr=&/Robert_Hit_Damaged_Light_Vocal_02}}
 {id=WavePlayer_26, properties {soundWaveAssetPtr=&Robert_Hit_Damaged_Light_Vocal_03}}
 {id=WavePlayer_27, properties {soundWaveAssetPtr=&Robert_Hit_Damaged_Light_Vocal_04}}
 {id=WavePlayer_28, properties {soundWaveAssetPtr=&Robert_Hit_Damaged_Light_Vocal_05}}
 {id=WavePlayer_29, properties {soundWaveAssetPtr=&Robert_Hit_Damaged_Light_Vocal_06}}
 {id=WavePlayer_30, properties {soundWaveAssetPtr=&Robert_Hit_Damaged_Light_Vocal_07}}
 {id=WavePlayer_31, properties {soundWaveAssetPtr=&Robert_Hit_Damaged_Light_Vocal_08}}
 {id=WavePlayer_8, properties {soundWaveAssetPtr=&Robert_Hit_Damaged_Light_Vocal_01}}
 {id=WavePlayer_9, properties {soundWaveAssetPtr=&Robert_Hit_Damaged_Light_Vocal_02}}
]
properties {
 attenuationSettings=/Irwin/Sounds/Settings/Predator_Vocalization_Medium_Att
 bOverrideConcurrency=true
 maxDistance=12000
 soundClassObject=/Game/Sounds/FortSoundClassesAndModes/Fort_Creatures_Vocalization_Normal
 volumeMultiplier=1.25
 concurrencyOverrides {bLimitToOwner=true, maxCount=1, resolutionRule=PreventNew}
}
