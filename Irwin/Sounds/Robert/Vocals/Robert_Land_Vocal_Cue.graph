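# Sound cue for Robert's landing vocal, drawing on reused Turn_90 vocal clips, with medium predator attenuation and one voice per owner.
# An IsLocal branch plays full volume when true; otherwise IsRiddenByEnemy selects full volume when true or unset, half when false.
# Each path randomizes the same 19 clips.
# The local and half-volume pools also hold empty weighted slots, so some triggers can be silent.
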SoundCue /Irwin/Sounds/Robert/Vocals/Robert_Land_Vocal_Cue
root=Branch_2
nodes [
 {id=Branch_0, children [Mixer_2,Mixer_1,Mixer_2], properties {boolParameterName=IsRiddenByEnemy}}
 {id=Branch_2, children [Mixer_3,Branch_0,Branch_0], properties {boolParameterName=IsLocal}}
 {id=Mixer_1, children [Random_1], properties {inputVolume [0.5]}}
 {id=Mixer_2, children [Random_0], properties {inputVolume [1]}}
 {id=Mixer_3, children [Random_2], properties {inputVolume [1]}}
 {
  id=Random_0
  children [
   WavePlayer_0
   WavePlayer_1
   WavePlayer_2
   WavePlayer_3
   WavePlayer_4
   WavePlayer_5
   WavePlayer_6
   WavePlayer_7
   WavePlayer_8
   WavePlayer_9
   WavePlayer_10
   WavePlayer_11
   WavePlayer_12
   WavePlayer_13
   WavePlayer_14
   WavePlayer_15
   WavePlayer_17
   WavePlayer_26
   WavePlayer_27
  ]
  properties {bShouldExcludeFromBranchCulling=true, weights [1,1,1,1,1,1,1,1,1,1,1,1,1,1,1,1,1,1,1]}
 }
 {
  id=Random_1
  children [
   WavePlayer_38
   WavePlayer_39
   WavePlayer_40
   WavePlayer_41
   WavePlayer_42
   WavePlayer_43
   WavePlayer_44
   WavePlayer_45
   WavePlayer_46
   WavePlayer_47
   WavePlayer_48
   WavePlayer_49
   WavePlayer_50
   WavePlayer_51
   WavePlayer_52
   WavePlayer_53
   WavePlayer_54
   WavePlayer_55
   WavePlayer_56
   null
   null
   null
   null
   null
   null
   null
   null
   null
  ]
  properties {bShouldExcludeFromBranchCulling=true, weights [1,1,1,1,1,1,1,1,1,1,1,1,1,1,1,1,1,1,1,1,1,1,1,1,1,1,1,1]}
 }
 {
  id=Random_2
  children [
   WavePlayer_16
   WavePlayer_18
   WavePlayer_19
   WavePlayer_20
   WavePlayer_21
   WavePlayer_22
   WavePlayer_23
   WavePlayer_24
   WavePlayer_25
   WavePlayer_28
   WavePlayer_29
   WavePlayer_30
   WavePlayer_31
   WavePlayer_32
   WavePlayer_33
   WavePlayer_34
   WavePlayer_35
   WavePlayer_36
   WavePlayer_37
   null
   null
   null
   null
  ]
  properties {bShouldExcludeFromBranchCulling=true, weights [1,1,1,1,1,1,1,1,1,1,1,1,1,1,1,1,1,1,1,1,1,1,1]}
 }
 {id=WavePlayer_0, properties {soundWaveAssetPtr=&/Robert_Turn_90_Vocal_20}}
 {id=WavePlayer_1, properties {soundWaveAssetPtr=&/Robert_Turn_90_Vocal_22}}
 {id=WavePlayer_10, properties {soundWaveAssetPtr=&/Robert_Turn_90_Vocal_33}}
 {id=WavePlayer_11, properties {soundWaveAssetPtr=&/Robert_Turn_90_Vocal_01}}
 {id=WavePlayer_12, properties {soundWaveAssetPtr=&/Robert_Turn_90_Vocal_02}}
 {id=WavePlayer_13, properties {soundWaveAssetPtr=&/Robert_Turn_90_Vocal_03}}
 {id=WavePlayer_14, properties {soundWaveAssetPtr=&/Robert_Turn_90_Vocal_04}}
 {id=WavePlayer_15, properties {soundWaveAssetPtr=&/Robert_Turn_90_Vocal_05}}
 {id=WavePlayer_16, properties {soundWaveAssetPtr=&Robert_Turn_90_Vocal_20}}
 {id=WavePlayer_17, properties {soundWaveAssetPtr=&/Robert_Turn_90_Vocal_06}}
 {id=WavePlayer_18, properties {soundWaveAssetPtr=&Robert_Turn_90_Vocal_22}}
 {id=WavePlayer_19, properties {soundWaveAssetPtr=&/Robert_Turn_90_Vocal_23}}
 {id=WavePlayer_2, properties {soundWaveAssetPtr=&Robert_Turn_90_Vocal_23}}
 {id=WavePlayer_20, properties {soundWaveAssetPtr=&/Robert_Turn_90_Vocal_24}}
 {id=WavePlayer_21, properties {soundWaveAssetPtr=&/Robert_Turn_90_Vocal_25}}
 {id=WavePlayer_22, properties {soundWaveAssetPtr=&/Robert_Turn_90_Vocal_26}}
 {id=WavePlayer_23, properties {soundWaveAssetPtr=&/Robert_Turn_90_Vocal_27}}
 {id=WavePlayer_24, properties {soundWaveAssetPtr=&/Robert_Turn_90_Vocal_28}}
 {id=WavePlayer_25, properties {soundWaveAssetPtr=&/Robert_Turn_90_Vocal_29}}
 {id=WavePlayer_26, properties {soundWaveAssetPtr=&/Robert_Turn_90_Vocal_07}}
 {id=WavePlayer_27, properties {soundWaveAssetPtr=&/Robert_Turn_90_Vocal_08}}
 {id=WavePlayer_28, properties {soundWaveAssetPtr=&/Robert_Turn_90_Vocal_32}}
 {id=WavePlayer_29, properties {soundWaveAssetPtr=&Robert_Turn_90_Vocal_33}}
 {id=WavePlayer_3, properties {soundWaveAssetPtr=&Robert_Turn_90_Vocal_24}}
 {id=WavePlayer_30, properties {soundWaveAssetPtr=&Robert_Turn_90_Vocal_01}}
 {id=WavePlayer_31, properties {soundWaveAssetPtr=&Robert_Turn_90_Vocal_02}}
 {id=WavePlayer_32, properties {soundWaveAssetPtr=&Robert_Turn_90_Vocal_03}}
 {id=WavePlayer_33, properties {soundWaveAssetPtr=&Robert_Turn_90_Vocal_04}}
 {id=WavePlayer_34, properties {soundWaveAssetPtr=&Robert_Turn_90_Vocal_05}}
 {id=WavePlayer_35, properties {soundWaveAssetPtr=&Robert_Turn_90_Vocal_06}}
 {id=WavePlayer_36, properties {soundWaveAssetPtr=&Robert_Turn_90_Vocal_07}}
 {id=WavePlayer_37, properties {soundWaveAssetPtr=&Robert_Turn_90_Vocal_08}}
 {id=WavePlayer_38, properties {soundWaveAssetPtr=&Robert_Turn_90_Vocal_20}}
 {id=WavePlayer_39, properties {soundWaveAssetPtr=&Robert_Turn_90_Vocal_22}}
 {id=WavePlayer_4, properties {soundWaveAssetPtr=&Robert_Turn_90_Vocal_25}}
 {id=WavePlayer_40, properties {soundWaveAssetPtr=&Robert_Turn_90_Vocal_23}}
 {id=WavePlayer_41, properties {soundWaveAssetPtr=&Robert_Turn_90_Vocal_24}}
 {id=WavePlayer_42, properties {soundWaveAssetPtr=&Robert_Turn_90_Vocal_25}}
 {id=WavePlayer_43, properties {soundWaveAssetPtr=&Robert_Turn_90_Vocal_26}}
 {id=WavePlayer_44, properties {soundWaveAssetPtr=&Robert_Turn_90_Vocal_27}}
 {id=WavePlayer_45, properties {soundWaveAssetPtr=&Robert_Turn_90_Vocal_28}}
 {id=WavePlayer_46, properties {soundWaveAssetPtr=&Robert_Turn_90_Vocal_29}}
 {id=WavePlayer_47, properties {soundWaveAssetPtr=&Robert_Turn_90_Vocal_32}}
 {id=WavePlayer_48, properties {soundWaveAssetPtr=&Robert_Turn_90_Vocal_33}}
 {id=WavePlayer_49, properties {soundWaveAssetPtr=&Robert_Turn_90_Vocal_01}}
 {id=WavePlayer_5, properties {soundWaveAssetPtr=&Robert_Turn_90_Vocal_26}}
 {id=WavePlayer_50, properties {soundWaveAssetPtr=&Robert_Turn_90_Vocal_02}}
 {id=WavePlayer_51, properties {soundWaveAssetPtr=&Robert_Turn_90_Vocal_03}}
 {id=WavePlayer_52, properties {soundWaveAssetPtr=&Robert_Turn_90_Vocal_04}}
 {id=WavePlayer_53, properties {soundWaveAssetPtr=&Robert_Turn_90_Vocal_05}}
 {id=WavePlayer_54, properties {soundWaveAssetPtr=&Robert_Turn_90_Vocal_06}}
 {id=WavePlayer_55, properties {soundWaveAssetPtr=&Robert_Turn_90_Vocal_07}}
 {id=WavePlayer_56, properties {soundWaveAssetPtr=&Robert_Turn_90_Vocal_08}}
 {id=WavePlayer_6, properties {soundWaveAssetPtr=&Robert_Turn_90_Vocal_27}}
 {id=WavePlayer_7, properties {soundWaveAssetPtr=&Robert_Turn_90_Vocal_28}}
 {id=WavePlayer_8, properties {soundWaveAssetPtr=&Robert_Turn_90_Vocal_29}}
 {id=WavePlayer_9, properties {soundWaveAssetPtr=&Robert_Turn_90_Vocal_32}}
]
properties {
 attenuationSettings=/Irwin/Sounds/Settings/Predator_Vocalization_Medium_Att
 bOverrideConcurrency=true
 maxDistance=12000
 soundClassObject=/Game/Sounds/FortSoundClassesAndModes/Fort_Creatures_Vocalization_Normal
 volumeMultiplier=1.25
 concurrencyOverrides {bLimitToOwner=true, maxCount=1}
}
