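# Sound cue for Robert's melee leap vocal, a creature vocalization using a medium predator attenuation and a capped-concurrency override.
# It branches on IsLocal, then on IsRiddenByEnemy when non-local or unset, with an unset rider state sharing the ridden-by-enemy path.
# Each of the three paths runs through its own mixer into an equal-weight random pick of eight takes, excluded from branch culling.
# All paths use identically named takes at the same mixer volume.
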SoundCue /Irwin/Sounds/Robert/Vocals/Robert_Melee_Leap_Vocal_Cue
root=Branch_2
nodes [
 {id=Branch_0, children [Mixer_2,Mixer_1,Mixer_2], properties {boolParameterName=IsRiddenByEnemy}}
 {id=Branch_2, children [Mixer_3,Branch_0,Branch_0], properties {boolParameterName=IsLocal}}
 {id=Mixer_1, children [Random_5], properties {inputVolume [1]}}
 {id=Mixer_2, children [Random_4], properties {inputVolume [1]}}
 {id=Mixer_3, children [Random_0], properties {inputVolume [1]}}
 {
  id=Random_0
  children [WavePlayer_0,WavePlayer_1,WavePlayer_2,WavePlayer_3,WavePlayer_4,WavePlayer_5,WavePlayer_6,WavePlayer_7]
  properties {bShouldExcludeFromBranchCulling=true, weights [1,1,1,1,1,1,1,1]}
 }
 {
  id=Random_4
  children [WavePlayer_27, WavePlayer_28, WavePlayer_29, WavePlayer_30, WavePlayer_31, WavePlayer_32, WavePlayer_33, WavePlayer_34]
  properties {bShouldExcludeFromBranchCulling=true, weights [1,1,1,1,1,1,1,1]}
 }
 {
  id=Random_5
  children [WavePlayer_35, WavePlayer_36, WavePlayer_37, WavePlayer_38, WavePlayer_39, WavePlayer_40, WavePlayer_41, WavePlayer_42]
  properties {bShouldExcludeFromBranchCulling=true, weights [1,1,1,1,1,1,1,1]}
 }
 {id=WavePlayer_0, properties {soundWaveAssetPtr=&/Robert_Melee_Leap_Vocal_01}}
 {id=WavePlayer_1, properties {soundWaveAssetPtr=&/Robert_Melee_Leap_Vocal_02}}
 {id=WavePlayer_2, properties {soundWaveAssetPtr=&/Robert_Melee_Leap_Vocal_03}}
 {id=WavePlayer_27, properties {soundWaveAssetPtr=&Robert_Melee_Leap_Vocal_01}}
 {id=WavePlayer_28, properties {soundWaveAssetPtr=&Robert_Melee_Leap_Vocal_02}}
 {id=WavePlayer_29, properties {soundWaveAssetPtr=&Robert_Melee_Leap_Vocal_03}}
 {id=WavePlayer_3, properties {soundWaveAssetPtr=&/Robert_Melee_Leap_Vocal_04}}
 {id=WavePlayer_30, properties {soundWaveAssetPtr=&Robert_Melee_Leap_Vocal_04}}
 {id=WavePlayer_31, properties {soundWaveAssetPtr=&/Robert_Melee_Leap_Vocal_05}}
 {id=WavePlayer_32, properties {soundWaveAssetPtr=&/Robert_Melee_Leap_Vocal_06}}
 {id=WavePlayer_33, properties {soundWaveAssetPtr=&/Robert_Melee_Leap_Vocal_07}}
 {id=WavePlayer_34, properties {soundWaveAssetPtr=&/Robert_Melee_Leap_Vocal_08}}
 {id=WavePlayer_35, properties {soundWaveAssetPtr=&Robert_Melee_Leap_Vocal_01}}
 {id=WavePlayer_36, properties {soundWaveAssetPtr=&Robert_Melee_Leap_Vocal_02}}
 {id=WavePlayer_37, properties {soundWaveAssetPtr=&Robert_Melee_Leap_Vocal_03}}
 {id=WavePlayer_38, properties {soundWaveAssetPtr=&Robert_Melee_Leap_Vocal_04}}
 {id=WavePlayer_39, properties {soundWaveAssetPtr=&Robert_Melee_Leap_Vocal_05}}
 {id=WavePlayer_4, properties {soundWaveAssetPtr=&Robert_Melee_Leap_Vocal_05}}
 {id=WavePlayer_40, properties {soundWaveAssetPtr=&Robert_Melee_Leap_Vocal_06}}
 {id=WavePlayer_41, properties {soundWaveAssetPtr=&Robert_Melee_Leap_Vocal_07}}
 {id=WavePlayer_42, properties {soundWaveAssetPtr=&Robert_Melee_Leap_Vocal_08}}
 {id=WavePlayer_5, properties {soundWaveAssetPtr=&Robert_Melee_Leap_Vocal_06}}
 {id=WavePlayer_6, properties {soundWaveAssetPtr=&Robert_Melee_Leap_Vocal_07}}
 {id=WavePlayer_7, properties {soundWaveAssetPtr=&Robert_Melee_Leap_Vocal_08}}
]
properties {
 attenuationSettings=/Irwin/Sounds/Settings/Predator_Vocalization_Medium_Att
 bOverrideConcurrency=true
 maxDistance=12000
 soundClassObject=&Fort_Creatures_Vocalization_Tactical
 volumeMultiplier=1.8
 concurrencyOverrides {maxCount=3}
}
